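# Animation Blueprint for the F_MED_Quarrel body mesh that follows its attached parent's pose and adds deformation fixes and bag jiggle.
# Copy Pose From Mesh feeds twist and sculpter Control Rigs, then constant bone offsets for strap, neck and collar, then spring-driven AnimDynamics on two bag bones.
# The gauntlet offset chain is disconnected from the output.
# Each update sets LookUp to zero; its head and arm angle calculations are unwired, and no node reads it.
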AnimBlueprint /BRCosmetics/Characters/Player/Female/Medium/Bodies/F_MED_Quarrel/Meshes/F_MED_Quarrel_AnimBP
target_skeleton=&/F_MED_Quarrel_Skeleton
graphs [
 AnimGraph {nodes [
   n1 = "Transform (Modify) Bone - Bone: dyn_l_gauntlet"(Alpha=1, Translation="3,0,0", Rotation="0,0,0", Scale="1,1,1") {
    rotationMode=BMM_Additive
    rotationSpace=BCS_BoneSpace
    scaleMode=BMM_Additive
    scaleSpace=BCS_BoneSpace
    translationMode=BMM_Additive
    translationSpace=BCS_BoneSpace
    alphaBoolBlend {blendOption=Linear}
    boneToModify {boneName=dyn_l_gauntlet}
    translation {x=3}
   }
   "Transform (Modify) Bone - Bone: dyn_r_gauntlet"(ComponentPose=@n1, Alpha=1, Translation="-3,0,0", Rotation="0,0,0", Scale="1,1,1") {
    rotationMode=BMM_Additive
    rotationSpace=BCS_BoneSpace
    scaleMode=BMM_Additive
    scaleSpace=BCS_BoneSpace
    translationMode=BMM_Additive
    translationSpace=BCS_BoneSpace
    alphaBoolBlend {blendOption=Linear}
    boneToModify {boneName=dyn_r_gauntlet}
    translation {x=-3}
   }
   n2 = Control Rig(Source=Copy Pose From Mesh() {bUseAttachedParent=true}) {
    alphaBoolBlend {blendOption=Linear}
    controlRigAssetReference {blueprintRigClass=&/F_MED_Quarrel_Deform_CtrlRig_twists.F_MED_Quarrel_Deform_CtrlRig_twists_C}
   }
   n3 = Control Rig(Source=@n2) {
    alphaBoolBlend {blendOption=Linear}
    controlRigAssetReference {blueprintRigClass=&/F_MED_Quarrel_Deform_CtrlRig_sculpters.F_MED_Quarrel_Deform_CtrlRig_sculpters_C}
   }
   n4 = "Transform (Modify) Bone - Bone: dyn_Strap_Expand"(ComponentPose=Local To Component(LocalPose=@n3), Alpha=1, Translation="0,0,0", Rotation="0,0,0", Scale="1.1,1.1,1.1") {
    rotationMode=BMM_Additive
    scaleMode=BMM_Additive
    translationMode=BMM_Additive
    alphaBoolBlend {blendOption=Linear}
    boneToModify {boneName=dyn_Strap_Expand}
    scale=<1.1 1.1 1.1>
   }
   n5 = "Transform (Modify) Bone - Bone: dyn_NeckFix"(ComponentPose=@n4, Alpha=1, Translation="0,0,0", Rotation="0,0,0", Scale="1.125,1.175,1.125") {
    rotationMode=BMM_Additive
    scaleMode=BMM_Additive
    translationMode=BMM_Additive
    alphaBoolBlend {blendOption=Linear}
    boneToModify {boneName=dyn_NeckFix}
    scale=<1.125 1.175 1.125>
   }
   n6 = "Transform (Modify) Bone - Bone: dyn_lookUp_CollarFix"(ComponentPose=@n5, Alpha=1, Translation="0,-5,0", Rotation="0,0,0", Scale="1,1,1") {
    rotationMode=BMM_Additive
    scaleMode=BMM_Additive
    translationMode=BMM_Additive
    alphaBoolBlend {blendOption=Linear}
    boneToModify {boneName=dyn_lookUp_CollarFix}
    translation {y=-5}
   }
   n7 = Anim Dynamics - Bone: dyn_bag_01(ComponentPose=@n6, Alpha=0.8, ExternalForce="0,0,0") {
    angularDampingOverride=0.92
    angularSpringConstant=60
    bAngularSpring=true
    bGravityOverrideInSimSpace=true
    bOverrideAngularDamping=true
    gravityScale=0.8
    simulationSpace=RootRelative
    alphaBoolBlend {blendOption=Linear}
    boundBone {boneName=dyn_bag_01}
    physicsBodyDefinitions [{
      sphereCollisionRadius=0
      boundBone {boneName=dyn_bag_01}
      boxExtents=<15 22 22>
      constraintSetup {angularLimitsMax=<20 10 5>, angularLimitsMin=<-20 -20 -5>, angularTarget=<0 0 0>}
      localJointOffset=<1.5 -0.05 0.1>
     }]
   }
   n8 = Anim Dynamics - Bone: dyn_bag_02(ComponentPose=@n7, Alpha=0.8, ExternalForce="0,0,0") {
    angularDampingOverride=0.9999
    angularSpringConstant=60
    bAngularSpring=true
    bGravityOverrideInSimSpace=true
    bOverrideAngularDamping=true
    gravityScale=0.8
    simulationSpace=RootRelative
    alphaBoolBlend {blendOption=Linear}
    boundBone {boneName=dyn_bag_02}
    physicsBodyDefinitions [{
      sphereCollisionRadius=0
      boundBone {boneName=dyn_bag_02}
      boxExtents=<15 22 22>
      constraintSetup {angularLimitsMax=<20 10 5>, angularLimitsMin=<-20 -20 -5>, angularTarget=<0 0 0>}
      localJointOffset=<1.5 -0.05 0.1>
     }]
   }
   Output Pose(Result=Component To Local(ComponentPose=@n8)) {layerGroup=None}
  ]}
 EventGraph {nodes [
   Event BlueprintUpdateAnimation()
   > Set LookUp(LookUp:double=0) -> (Output_Get:double)
   n1 = BreakTransform(InTransform=GetDeltaTransformFromRefPose(self=GetOwnerSkeletalMeshComponent(), BoneName=head, BaseName=neck_02))
   MapRangeClamped(Value="Absolute (Float)"(A=BreakRotator(InRot=@n1.Rotation).Yaw), InRangeA=0, InRangeB=90)
   n2 = BreakTransform(InTransform=GetDeltaTransformFromRefPose(self=GetOwnerSkeletalMeshComponent(), BoneName=upperarm_r, BaseName=clavicle_r))
   Set LookUp(LookUp=MapRangeClamped(Value=BreakRotator(InRot=@n2.Rotation).Yaw, InRangeA=90, InRangeB=120))
  ]}
]
variables [LookUp {type=number}]
