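# Physics asset for the FrontEnd variant of the Quarrel female character's face accessory, covering two dynamic hair braids.
# Each braid is an eight-bone capsule chain running from a master bone on the head, with swing limits and a tighter twist limit between links.
# Capsules on the root, pelvis, spine, neck, head, clavicles and arms form a simplified body with free angular joints, likely serving as collision for the braids.
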PhysicsAsset /BRCosmetics/Characters/Player/Female/Medium/Bodies/F_MED_Quarrel/Meshes/Parts/F_MED_Quarrel_FaceAcc_FrontEnd_Physics
bodies {
 clavicle_l [{shapeType=Capsule, length=0.505, radius=0.505, shapeName=clavicle_l_capsule, rotation=<rot 0 -90 0>}]
 clavicle_r [{shapeType=Capsule, length=0.505, radius=0.505, shapeName=clavicle_r_capsule, rotation=<rot 0 -90 0>}]
 dyn_l_braid_01 [{shapeType=Capsule, length=5.5765, radius=4.0213, shapeName=dyn_l_braid_01_capsule, center=<5.3916 0.0002 3>, rotation=<rot 90 0 0>}]
 dyn_l_braid_02 [{shapeType=Capsule, length=5.1663, radius=4.0375, shapeName=dyn_l_braid_02_capsule, center=<5.3916 0.0002 3>, rotation=<rot 90 0 0>}]
 dyn_l_braid_03 [{shapeType=Capsule, length=4.7692, radius=3.9139, shapeName=dyn_l_braid_03_capsule, center=<5.3916 0.0002 3>, rotation=<rot 90 0 0>}]
 dyn_l_braid_04 [{shapeType=Capsule, length=4.6634, radius=4.2263, shapeName=dyn_l_braid_04_capsule, center=<5.3916 0.0002 3>, rotation=<rot 90 0 0>}]
 dyn_l_braid_05 [{shapeType=Capsule, length=5.1006, radius=4.233, shapeName=dyn_l_braid_05_capsule, center=<5.3916 0.0002 3>, rotation=<rot 90 0 0>}]
 dyn_l_braid_06 [{shapeType=Capsule, length=5.4468, radius=3.9189, shapeName=dyn_l_braid_06_capsule, center=<5.3916 0.0002 3>, rotation=<rot 90 0 0>}]
 dyn_l_braid_07 [{shapeType=Capsule, length=2.4691, radius=4.6045, shapeName=dyn_l_braid_07_capsule, center=<4.3641 0.0002 3>, rotation=<rot 90 0 0>}]
 dyn_l_braid_08 [{shapeType=Capsule, length=2.8608, radius=5.1277, shapeName=dyn_l_braid_08_capsule, center=<5.3916 0.0002 3>, rotation=<rot 90 0 0>}]
 dyn_l_braid_MSTR [{shapeType=Capsule, length=0.505, radius=0.505, shapeName=dyn_l_braid_MSTR_capsule}]
 dyn_r_braid_01 [{shapeType=Capsule, length=5.1613, radius=4.063, shapeName=dyn_r_braid_01_capsule, center=<5.3916 0.0002 3>, rotation=<rot 90 0 0>}]
 dyn_r_braid_02 [{shapeType=Capsule, length=5.4184, radius=4.0515, shapeName=dyn_r_braid_02_capsule, center=<5.3916 0.0002 3>, rotation=<rot 90 0 0>}]
 dyn_r_braid_03 [{shapeType=Capsule, length=4.9955, radius=3.9228, shapeName=dyn_r_braid_03_capsule, center=<5.3916 0.0002 3>, rotation=<rot 90 0 0>}]
 dyn_r_braid_04 [{shapeType=Capsule, length=5.234, radius=4.2521, shapeName=dyn_r_braid_04_capsule, center=<5.3916 0.0002 3>, rotation=<rot 90 0 0>}]
 dyn_r_braid_05 [{shapeType=Capsule, length=4.548, radius=4.2078, shapeName=dyn_r_braid_05_capsule, center=<5.3916 0.0002 3>, rotation=<rot 90 0 0>}]
 dyn_r_braid_06 [{shapeType=Capsule, length=5.151, radius=4.0841, shapeName=dyn_r_braid_06_capsule, center=<5.3916 0.0002 3>, rotation=<rot 90 0 0>}]
 dyn_r_braid_07 [{shapeType=Capsule, length=2.2491, radius=4.4505, shapeName=dyn_r_braid_07_capsule, center=<4.3248 0.0002 3>, rotation=<rot 90 0 0>}]
 dyn_r_braid_08 [{shapeType=Capsule, length=2.005, radius=5.1528, shapeName=dyn_r_braid_08_capsule, center=<5.1131 0.0001 3>, rotation=<rot 90 0 0>}]
 dyn_r_braid_MSTR [{shapeType=Capsule, length=0.505, radius=0.505, shapeName=dyn_r_braid_MSTR_capsule}]
 head [{shapeType=Capsule, length=4.1307, radius=8.6857, shapeName=head_capsule, center=<6.497 0 0.0004>, rotation=<rot 90 0 0>}]
 lowerarm_l [{shapeType=Capsule, length=17.7962, radius=6.0861, shapeName=lowerarm_l_capsule, center=<13.5564 0.0001 0>, rotation=<rot 90 0 0>}]
 lowerarm_r [{shapeType=Capsule, length=14.4483, radius=6.2718, shapeName=lowerarm_r_capsule, center=<-12.5886 0.0001 0>, rotation=<rot 90 0 0>}]
 neck_01 [{shapeType=Capsule, length=5.0825, radius=5.1345, shapeName=neck_01_capsule}]
 neck_02 [{shapeType=Capsule, length=5.0825, radius=5.1345, shapeName=neck_02_capsule}]
 pelvis [{shapeType=Capsule, length=20, radius=15, shapeName=pelvis_capsule}]
 root [{shapeType=Capsule, length=0.505, radius=0.505, shapeName=root_capsule, rotation=<rot 0 -90 0>}]
 spine_01 [{shapeType=Capsule, length=20, radius=15, shapeName=spine_01_capsule}]
 spine_02 [{shapeType=Capsule, length=20.4428, radius=14.5572, shapeName=spine_02_capsule}]
 spine_03 [{shapeType=Capsule, length=21.8933, radius=13.1067, shapeName=spine_03_capsule}]
 spine_04 [{shapeType=Capsule, length=23.0422, radius=11.9578, shapeName=spine_04_capsule}]
 spine_05 [{shapeType=Capsule, length=24.1622, radius=10.8377, shapeName=spine_05_capsule}]
 upperarm_l [{shapeType=Capsule, length=17.2496, radius=5.1192, shapeName=upperarm_l_capsule, center=<12.4278 -0.0001 0>, rotation=<rot 90 0 0>}]
 upperarm_r [{shapeType=Capsule, length=15.8436, radius=5.3356, shapeName=upperarm_r_capsule, center=<-14.7963 2.3062 -0.0002>, rotation=<rot 90 0 0>}]
}
constraints [
 {
  bone1Name=dyn_l_braid_MSTR
  bone2Name=head
  swing1LimitDegrees=45
  swing1Motion=Free
  swing2LimitDegrees=45
  swing2Motion=Free
  twistLimitDegrees=45
  twistMotion=Free
 }
 {
  bone1Name=dyn_r_braid_MSTR
  bone2Name=head
  swing1LimitDegrees=45
  swing1Motion=Free
  swing2LimitDegrees=45
  swing2Motion=Free
  twistLimitDegrees=45
  twistMotion=Free
 }
 {bone1Name=dyn_l_braid_01, bone2Name=dyn_l_braid_MSTR, swing1LimitDegrees=45, swing2LimitDegrees=45, twistLimitDegrees=5}
 {bone1Name=dyn_l_braid_02, bone2Name=dyn_l_braid_01, swing1LimitDegrees=45, swing2LimitDegrees=45, twistLimitDegrees=5}
 {bone1Name=dyn_l_braid_03, bone2Name=dyn_l_braid_02, swing1LimitDegrees=45, swing2LimitDegrees=45, twistLimitDegrees=5}
 {bone1Name=dyn_l_braid_04, bone2Name=dyn_l_braid_03, swing1LimitDegrees=45, swing2LimitDegrees=45, twistLimitDegrees=5}
 {bone1Name=dyn_l_braid_05, bone2Name=dyn_l_braid_04, swing1LimitDegrees=45, swing2LimitDegrees=45, twistLimitDegrees=5}
 {bone1Name=dyn_l_braid_06, bone2Name=dyn_l_braid_05, swing1LimitDegrees=45, swing2LimitDegrees=45, twistLimitDegrees=5}
 {bone1Name=dyn_l_braid_07, bone2Name=dyn_l_braid_06, swing1LimitDegrees=45, swing2LimitDegrees=45, twistLimitDegrees=5}
 {bone1Name=dyn_l_braid_08, bone2Name=dyn_l_braid_07, swing1LimitDegrees=45, swing2LimitDegrees=45, twistLimitDegrees=5}
 {bone1Name=dyn_r_braid_01, bone2Name=dyn_r_braid_MSTR, swing1LimitDegrees=45, swing2LimitDegrees=45, twistLimitDegrees=5}
 {bone1Name=dyn_r_braid_02, bone2Name=dyn_r_braid_01, swing1LimitDegrees=45, swing2LimitDegrees=45, twistLimitDegrees=5}
 {bone1Name=dyn_r_braid_03, bone2Name=dyn_r_braid_02, swing1LimitDegrees=45, swing2LimitDegrees=45, twistLimitDegrees=5}
 {bone1Name=dyn_r_braid_04, bone2Name=dyn_r_braid_03, swing1LimitDegrees=45, swing2LimitDegrees=45, twistLimitDegrees=5}
 {bone1Name=dyn_r_braid_05, bone2Name=dyn_r_braid_04, swing1LimitDegrees=45, swing2LimitDegrees=45, twistLimitDegrees=5}
 {bone1Name=dyn_r_braid_06, bone2Name=dyn_r_braid_05, swing1LimitDegrees=45, swing2LimitDegrees=45, twistLimitDegrees=5}
 {bone1Name=dyn_r_braid_07, bone2Name=dyn_r_braid_06, swing1LimitDegrees=45, swing2LimitDegrees=45, twistLimitDegrees=5}
 {bone1Name=dyn_r_braid_08, bone2Name=dyn_r_braid_07, swing1LimitDegrees=45, swing2LimitDegrees=45, twistLimitDegrees=5}
 {
  bone1Name=head
  bone2Name=neck_02
  swing1LimitDegrees=45
  swing1Motion=Free
  swing2LimitDegrees=45
  swing2Motion=Free
  twistLimitDegrees=45
  twistMotion=Free
 }
 {
  bone1Name=spine_01
  bone2Name=pelvis
  swing1LimitDegrees=45
  swing1Motion=Free
  swing2LimitDegrees=45
  swing2Motion=Free
  twistLimitDegrees=45
  twistMotion=Free
 }
 {
  bone1Name=spine_02
  bone2Name=spine_01
  swing1LimitDegrees=45
  swing1Motion=Free
  swing2LimitDegrees=45
  swing2Motion=Free
  twistLimitDegrees=45
  twistMotion=Free
 }
 {
  bone1Name=spine_03
  bone2Name=spine_02
  swing1LimitDegrees=45
  swing1Motion=Free
  swing2LimitDegrees=45
  swing2Motion=Free
  twistLimitDegrees=45
  twistMotion=Free
 }
 {
  bone1Name=spine_04
  bone2Name=spine_03
  swing1LimitDegrees=45
  swing1Motion=Free
  swing2LimitDegrees=45
  swing2Motion=Free
  twistLimitDegrees=45
  twistMotion=Free
 }
 {
  bone1Name=spine_05
  bone2Name=spine_04
  swing1LimitDegrees=45
  swing1Motion=Free
  swing2LimitDegrees=45
  swing2Motion=Free
  twistLimitDegrees=45
  twistMotion=Free
 }
 {
  bone1Name=neck_01
  bone2Name=spine_05
  swing1LimitDegrees=45
  swing1Motion=Free
  swing2LimitDegrees=45
  swing2Motion=Free
  twistLimitDegrees=45
  twistMotion=Free
 }
 {
  bone1Name=neck_02
  bone2Name=neck_01
  swing1LimitDegrees=45
  swing1Motion=Free
  swing2LimitDegrees=45
  swing2Motion=Free
  twistLimitDegrees=45
  twistMotion=Free
 }
 {
  bone1Name=upperarm_l
  bone2Name=clavicle_l
  swing1LimitDegrees=45
  swing1Motion=Free
  swing2LimitDegrees=45
  swing2Motion=Free
  twistLimitDegrees=45
  twistMotion=Free
 }
 {
  bone1Name=upperarm_r
  bone2Name=clavicle_r
  swing1LimitDegrees=45
  swing1Motion=Free
  swing2LimitDegrees=45
  swing2Motion=Free
  twistLimitDegrees=45
  twistMotion=Free
 }
 {
  bone1Name=lowerarm_l
  bone2Name=upperarm_l
  swing1LimitDegrees=45
  swing1Motion=Free
  swing2LimitDegrees=45
  swing2Motion=Free
  twistLimitDegrees=45
  twistMotion=Free
 }
 {
  bone1Name=lowerarm_r
  bone2Name=upperarm_r
  swing1LimitDegrees=45
  swing1Motion=Free
  swing2LimitDegrees=45
  swing2Motion=Free
  twistLimitDegrees=45
  twistMotion=Free
 }
 {
  bone1Name=pelvis
  bone2Name=root
  swing1LimitDegrees=45
  swing1Motion=Free
  swing2LimitDegrees=45
  swing2Motion=Free
  twistLimitDegrees=45
  twistMotion=Free
 }
 {
  bone1Name=clavicle_l
  bone2Name=spine_05
  swing1LimitDegrees=45
  swing1Motion=Free
  swing2LimitDegrees=45
  swing2Motion=Free
  twistLimitDegrees=45
  twistMotion=Free
 }
 {
  bone1Name=clavicle_r
  bone2Name=spine_05
  swing1LimitDegrees=45
  swing1Motion=Free
  swing2LimitDegrees=45
  swing2Motion=Free
  twistLimitDegrees=45
  twistMotion=Free
 }
]
