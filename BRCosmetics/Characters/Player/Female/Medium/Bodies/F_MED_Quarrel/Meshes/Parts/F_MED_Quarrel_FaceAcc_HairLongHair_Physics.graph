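# Physics asset for the long-hair mesh part of the female Quarrel character, defining capsule bodies for bang simulation.
# A large head capsule anchors the left and right bang master bodies through constraints with free swing and twist.
# Only the right bang continues into a four-segment capsule chain, linked from its master with moderate swing limits and tight twist limits.
# The left bang has only its master body.
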PhysicsAsset /BRCosmetics/Characters/Player/Female/Medium/Bodies/F_MED_Quarrel/Meshes/Parts/F_MED_Quarrel_FaceAcc_HairLongHair_Physics
bodies {
 dyn_l_bang_MSTR [{shapeType=Capsule, length=0.505, radius=0.505, shapeName=dyn_l_bang_MSTR_capsule}]
 dyn_r_bang_01 [{shapeType=Capsule, length=3.33, radius=0.7207, shapeName=dyn_r_bang_01_capsule, center=<2.307 0 0>, rotation=<rot 90 0 0>}]
 dyn_r_bang_02 [{shapeType=Capsule, length=3.545, radius=0.6119, shapeName=dyn_r_bang_02_capsule, center=<2.3539 0 0>, rotation=<rot 90 0 0>}]
 dyn_r_bang_03 [{shapeType=Capsule, length=2.2853, radius=1.0735, shapeName=dyn_r_bang_03_capsule, center=<1.9354 0.0001 0>, rotation=<rot 90 0 0>}]
 dyn_r_bang_04 [{shapeType=Capsule, length=2.885, radius=1.0643, shapeName=dyn_r_bang_04_capsule, center=<1.7774 0 -0.4219>, rotation=<rot 75.8832 0 -0.0001>}]
 dyn_r_bang_MSTR [{shapeType=Capsule, length=0.505, radius=0.505, shapeName=dyn_r_bang_MSTR_capsule}]
 head [{shapeType=Capsule, length=3.2619, radius=7.7384, center=<0.9836 0.9975 0>}]
}
constraints [
 {
  bone1Name=dyn_r_bang_MSTR
  bone2Name=head
  swing1LimitDegrees=45
  swing1Motion=Free
  swing2LimitDegrees=45
  swing2Motion=Free
  twistLimitDegrees=45
  twistMotion=Free
 }
 {
  bone1Name=dyn_l_bang_MSTR
  bone2Name=head
  swing1LimitDegrees=45
  swing1Motion=Free
  swing2LimitDegrees=45
  swing2Motion=Free
  twistLimitDegrees=45
  twistMotion=Free
 }
 {bone1Name=dyn_r_bang_01, bone2Name=dyn_r_bang_MSTR, swing1LimitDegrees=45, swing2LimitDegrees=45, twistLimitDegrees=5}
 {bone1Name=dyn_r_bang_02, bone2Name=dyn_r_bang_01, swing1LimitDegrees=45, swing2LimitDegrees=45, twistLimitDegrees=5}
 {bone1Name=dyn_r_bang_03, bone2Name=dyn_r_bang_02, swing1LimitDegrees=45, swing2LimitDegrees=45, twistLimitDegrees=5}
 {bone1Name=dyn_r_bang_04, bone2Name=dyn_r_bang_03, swing1LimitDegrees=45, swing2LimitDegrees=45, twistLimitDegrees=5}
]
