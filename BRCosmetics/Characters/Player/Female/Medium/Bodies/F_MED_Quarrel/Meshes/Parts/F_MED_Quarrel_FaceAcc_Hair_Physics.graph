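# Physics asset for the left and right hair bangs on the Quarrel female medium character's face accessory hair mesh.
# Each bang is a chain of capsule bodies, three segments on the left and four on the right, hanging from a small master body constrained to a large, near-spherical head capsule.
# The master-to-head joints rotate freely, while the segment joints carry moderate swing and narrow twist limits.
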PhysicsAsset /BRCosmetics/Characters/Player/Female/Medium/Bodies/F_MED_Quarrel/Meshes/Parts/F_MED_Quarrel_FaceAcc_Hair_Physics
bodies {
 dyn_l_bang_01 [{shapeType=Capsule, length=1, radius=1, shapeName=dyn_l_bang_01_capsule, center=<1.2764 0 0>, rotation=<rot 90 0 0>}]
 dyn_l_bang_02 [{shapeType=Capsule, length=1, radius=1, shapeName=dyn_l_bang_02_capsule, center=<1.446 0 0>, rotation=<rot 90 0 0>}]
 dyn_l_bang_03 [{shapeType=Capsule, length=1, radius=1, shapeName=dyn_l_bang_03_capsule, center=<1.4927 0 0>, rotation=<rot 90 0 0>}]
 dyn_l_bang_MSTR [{shapeType=Capsule, length=0.505, radius=0.505, shapeName=dyn_l_bang_MSTR_capsule}]
 dyn_r_bang_01 [{shapeType=Capsule, length=2.9278, radius=1, shapeName=dyn_r_bang_01_capsule, center=<2.3341 0 0>, rotation=<rot 90 0 0>}]
 dyn_r_bang_02 [{shapeType=Capsule, length=2.5558, radius=1, shapeName=dyn_r_bang_02_capsule, center=<2.3539 0 0>, rotation=<rot 90 0 0>}]
 dyn_r_bang_03 [{shapeType=Capsule, length=2.0539, radius=1, shapeName=dyn_r_bang_03_capsule, center=<1.9354 0.0001 0>, rotation=<rot 90 0 0>}]
 dyn_r_bang_04 [{shapeType=Capsule, length=1.6837, radius=1, shapeName=dyn_r_bang_04_capsule, center=<1.8835 0 0>, rotation=<rot 90 0 0>}]
 dyn_r_bang_MSTR [{shapeType=Capsule, length=0.505, radius=0.505, shapeName=dyn_r_bang_MSTR_capsule}]
 head [{shapeType=Capsule, length=0.0003, radius=11, center=<0 0.9975 0>}]
}
constraints [
 {
  bone1Name=dyn_r_bang_MSTR
  bone2Name=head
  swing1LimitDegrees=45
  swing1Motion=Free
  swing2LimitDegrees=45
  swing2Motion=Free
  twistLimitDegrees=45
  twistMotion=Free
 }
 {
  bone1Name=dyn_l_bang_MSTR
  bone2Name=head
  swing1LimitDegrees=45
  swing1Motion=Free
  swing2LimitDegrees=45
  swing2Motion=Free
  twistLimitDegrees=45
  twistMotion=Free
 }
 {bone1Name=dyn_r_bang_01, bone2Name=dyn_r_bang_MSTR, swing1LimitDegrees=30, swing2LimitDegrees=30, twistLimitDegrees=5}
 {bone1Name=dyn_r_bang_02, bone2Name=dyn_r_bang_01, swing1LimitDegrees=30, swing2LimitDegrees=30, twistLimitDegrees=5}
 {bone1Name=dyn_r_bang_03, bone2Name=dyn_r_bang_02, swing1LimitDegrees=30, swing2LimitDegrees=30, twistLimitDegrees=5}
 {bone1Name=dyn_r_bang_04, bone2Name=dyn_r_bang_03, swing1LimitDegrees=30, swing2LimitDegrees=30, twistLimitDegrees=5}
 {bone1Name=dyn_l_bang_01, bone2Name=dyn_l_bang_MSTR, swing1LimitDegrees=30, swing2LimitDegrees=30, twistLimitDegrees=5}
 {bone1Name=dyn_l_bang_02, bone2Name=dyn_l_bang_01, swing1LimitDegrees=30, swing2LimitDegrees=30, twistLimitDegrees=5}
 {bone1Name=dyn_l_bang_03, bone2Name=dyn_l_bang_02, swing1LimitDegrees=30, swing2LimitDegrees=30, twistLimitDegrees=5}
]
