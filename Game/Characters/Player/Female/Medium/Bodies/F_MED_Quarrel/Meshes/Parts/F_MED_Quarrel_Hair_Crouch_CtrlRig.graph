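# Control Rig for the female Quarrel hair, named for crouch, that bends the left and right braid chains along Bezier curves.
# Curve points come from ponytail CV bones parented under head, upper spine and pelvis.
# Four chained FitChainToCurve units partially blend each braid toward a shared centre curve, then its side curve, with differing weights.
# Both right-braid fits feed Scale3D, not Translation, into curve points C and D, unlike the left.
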ControlRigBlueprint /Game/Characters/Player/Female/Medium/Bodies/F_MED_Quarrel/Meshes/Parts/F_MED_Quarrel_Hair_Crouch_CtrlRig
graphs [
 RigVMFunctionLibrary {links [], nodes []}
 RigVMModel {nodes [
   n1 = &RigUnit_GetBoneTransform(Bone=ponytail_CV_0)
   n2 = &RigUnit_GetBoneTransform(Bone=ponytail_CV_2)
   n3 = &RigUnit_GetBoneTransform(Bone=ponytail_CV_1)
   n4 = &RigUnit_GetBoneTransform(Bone=ponytail_CV_3)
   /Script/ControlRig.RigUnit_FitChainToCurve(
    Alignment=Front
    Bezier=(A=(X=0,Y=0,Z=0),B=(X=0,Y=0,Z=0),C=(X=0,Y=0,Z=0),D=(X=0,Y=0,Z=0))
    DebugSettings=(bEnabled=False,Scale=1,CurveColor=(R=1,G=1,B=0,A=1),SegmentsColor=(R=1,G=0,B=0,A=1),WorldOffset=(Rotation=(X=0,Y=0,Z=0,W=1),Translation=(X=0,Y=0,Z=0),Scale3D=(X=1,Y=1,Z=1)))
    EndBone=dyn_l_braid_08
    ExecutePin=()
    Maximum=1
    Minimum=0
    PoleVectorPosition=(X=0,Y=0,Z=0)
    PrimaryAxis=(X=1,Y=0,Z=0)
    RotationEaseType=Linear
    Rotations=()
    SamplingPrecision=12
    SecondaryAxis=(X=0,Y=0,Z=0)
    StartBone=dyn_l_braid_01
    Weight=0.75
    bPropagateToChildren=True
    Bezier.A=@n1.Transform.Translation
    Bezier.C=@n2.Transform.Translation
    Bezier.B=@n3.Transform.Translation
    Bezier.D=@n4.Transform.Translation
   ) {is=struct RigVMUnitNode}
   > &RigUnit_FitChainToCurve(
    Alignment=Front
    Bezier=(A=(X=0,Y=0,Z=0),B=(X=0,Y=0,Z=0),C=(X=0,Y=0,Z=0),D=(X=0,Y=0,Z=0))
    DebugSettings=(bEnabled=False,Scale=1,CurveColor=(R=1,G=1,B=0,A=1),SegmentsColor=(R=1,G=0,B=0,A=1),WorldOffset=(Rotation=(X=0,Y=0,Z=0,W=1),Translation=(X=0,Y=0,Z=0),Scale3D=(X=1,Y=1,Z=1)))
    EndBone=dyn_r_braid_08
    Maximum=1
    Minimum=0
    PoleVectorPosition=(X=0,Y=0,Z=0)
    PrimaryAxis=(X=1,Y=0,Z=0)
    RotationEaseType=Linear
    Rotations=()
    SamplingPrecision=12
    SecondaryAxis=(X=0,Y=0,Z=0)
    StartBone=dyn_r_braid_01
    Weight=0.75
    bPropagateToChildren=True
    Bezier.A=@n1.Transform.Translation
    Bezier.C=@n2.Transform.Scale3D
    Bezier.B=@n3.Transform.Translation
    Bezier.D=@n4.Transform.Scale3D
   )
   > &RigUnit_FitChainToCurve(
    Alignment=Front
    Bezier=(A=(X=0,Y=0,Z=0),B=(X=0,Y=0,Z=0),C=(X=0,Y=0,Z=0),D=(X=0,Y=0,Z=0))
    DebugSettings=(bEnabled=False,Scale=1,CurveColor=(R=1,G=1,B=0,A=1),SegmentsColor=(R=1,G=0,B=0,A=1),WorldOffset=(Rotation=(X=0,Y=0,Z=0,W=1),Translation=(X=0,Y=0,Z=0),Scale3D=(X=1,Y=1,Z=1)))
    EndBone=dyn_l_braid_08
    Maximum=1
    Minimum=0
    PoleVectorPosition=(X=0,Y=0,Z=0)
    PrimaryAxis=(X=1,Y=0,Z=0)
    RotationEaseType=Linear
    Rotations=()
    SamplingPrecision=12
    SecondaryAxis=(X=0,Y=0,Z=0)
    StartBone=dyn_l_braid_01
    Weight=0.25
    bPropagateToChildren=True
    Bezier.D=&RigUnit_GetBoneTransform(Bone=ponytail_CV_l_3).Transform.Translation
    Bezier.B=&RigUnit_GetBoneTransform(Bone=ponytail_CV_l_1).Transform.Translation
    Bezier.C=&RigUnit_GetBoneTransform(Bone=ponytail_CV_l_2).Transform.Translation
    Bezier.A=&RigUnit_GetBoneTransform(Bone=ponytail_CV_l_0).Transform.Translation
   )
   > &RigUnit_FitChainToCurve(
    Alignment=Front
    Bezier=(A=(X=0,Y=0,Z=0),B=(X=0,Y=0,Z=0),C=(X=0,Y=0,Z=0),D=(X=0,Y=0,Z=0))
    DebugSettings=(bEnabled=False,Scale=1,CurveColor=(R=1,G=1,B=0,A=1),SegmentsColor=(R=1,G=0,B=0,A=1),WorldOffset=(Rotation=(X=0,Y=0,Z=0,W=1),Translation=(X=0,Y=0,Z=0),Scale3D=(X=1,Y=1,Z=1)))
    EndBone=dyn_r_braid_08
    Maximum=1
    Minimum=0
    PoleVectorPosition=(X=0,Y=0,Z=0)
    PrimaryAxis=(X=1,Y=0,Z=0)
    RotationEaseType=Linear
    Rotations=()
    SamplingPrecision=12
    SecondaryAxis=(X=0,Y=0,Z=0)
    StartBone=dyn_r_braid_01
    Weight=0.5
    bPropagateToChildren=True
    Bezier.D=&RigUnit_GetBoneTransform(Bone=ponytail_CV_r_3).Transform.Scale3D
    Bezier.B=&RigUnit_GetBoneTransform(Bone=ponytail_CV_r_1).Transform.Translation
    Bezier.C=&RigUnit_GetBoneTransform(Bone=ponytail_CV_r_2).Transform.Scale3D
    Bezier.A=&RigUnit_GetBoneTransform(Bone=ponytail_CV_r_0).Transform.Translation
   )
   RigUnit_BeginExecution = &RigUnit_BeginExecution()
  ]}
]
hierarchy [
 root {loc=<0 0 0>, rot=<0 0 0 1>}
 pelvis {parent=root, loc=<0 -0.5301 98.0571>, rot=<0.0127 0.707 -0.0127 -0.707>}
 spine_01 {parent=pelvis, loc=<6.1208 -0.9328 0>, rot=<0 0 0.034 -0.9994>}
 spine_02 {parent=spine_01, loc=<6.5996 -0.2095 0>, rot=<0 0 0.0238 -0.9997>}
 spine_03 {parent=spine_02, loc=<6.4914 -0.1237 0>, rot=<0 0 0.0149 -0.9999>}
 spine_04 {parent=spine_03, loc=<6.9598 0.0794 0>, rot=<0 0 -0.0132 -0.9999>}
 spine_05 {parent=spine_04, loc=<6.9987 0 0>, rot=<0 0 -0.0015 -1>}
 clavicle_l {parent=spine_05, loc=<5.4831 -0.1053 -3.2946>, rot=<0.0465 -0.7289 -0.009 -0.6829>}
 upperarm_l {parent=clavicle_l, loc=<9.8658 -1.2136 -0.4105>, rot=<0.0208 -0.486 -0.0242 -0.8734>}
 lowerarm_l {parent=upperarm_l, loc=<26.465 -0.8345 0.259>, rot=<0.0036 0.0161 -0.2298 -0.9731>}
 hand_l {parent=lowerarm_l, loc=<22.7355 -0.3203 -0.6106>, rot=<-0.6784 -0.0223 0.0735 -0.7307>}
 clavicle_r {parent=spine_05, loc=<5.4831 -0.1053 3.2946>, rot=<0.7289 0.0465 -0.6829 0.009>}
 upperarm_r {parent=clavicle_r, loc=<-9.8658 1.2136 0.4104>, rot=<0.0208 -0.486 -0.0242 -0.8734>}
 lowerarm_r {parent=upperarm_r, loc=<-26.4651 0.8345 -0.259>, rot=<0.0036 0.0161 -0.2298 -0.9731>}
 hand_r {parent=lowerarm_r, loc=<-22.7356 0.3203 0.6106>, rot=<-0.6784 -0.0223 0.0735 -0.7307>}
 neck_01 {parent=spine_05, loc=<11.3138 0.1529 0>, rot=<0 0 -0.2003 -0.9797>}
 neck_02 {parent=neck_01, loc=<4.5627 0 0>, rot=<0 0 0.0504 -0.9987>}
 head {parent=neck_02, loc=<5.3914 0 0>, rot=<0 0 0.115 -0.9934>}
 dyn_l_braid_MSTR {parent=head, loc=<-0.5872 -6.3197 -4.2725>, rot=<0.0804 -0.6745 0.683 -0.2684>}
 dyn_l_braid_01 {parent=dyn_l_braid_MSTR, loc=<0 0 0>, rot=<0 0 0 1>}
 dyn_l_braid_02 {parent=dyn_l_braid_01, loc=<10.3206 0 0>, rot=<-0.0182 -0.1108 0.0109 -0.9936>}
 dyn_l_braid_03 {parent=dyn_l_braid_02, loc=<9.4429 0 0>, rot=<0.0244 -0.0545 0.0074 -0.9982>}
 dyn_l_braid_04 {parent=dyn_l_braid_03, loc=<9.0558 0 0>, rot=<-0.0031 -0.0311 -0.023 -0.9992>}
 dyn_l_braid_05 {parent=dyn_l_braid_04, loc=<9.7886 0 0>, rot=<0.0166 -0.0159 0.0079 -0.9997>}
 dyn_l_braid_06 {parent=dyn_l_braid_05, loc=<9.6982 0 0>, rot=<0.0103 -0.0164 -0.0144 -0.9997>}
 dyn_l_braid_07 {parent=dyn_l_braid_06, loc=<10.1948 0 0>, rot=<-0.0135 0.0198 0.0186 -0.9995>}
 dyn_l_braid_08 {parent=dyn_l_braid_07, loc=<7.54 0 0>, rot=<-0.0424 -0.0214 0.0327 -0.9983>}
 dyn_r_braid_MSTR {parent=head, loc=<-0.4499 -6.3061 4.2844>, rot=<0.2612 -0.6881 0.6714 -0.0871>}
 dyn_r_braid_01 {parent=dyn_r_braid_MSTR, loc=<0 0 0>, rot=<0 0 0 -1>}
 dyn_r_braid_02 {parent=dyn_r_braid_01, loc=<10.3206 0 0>, rot=<-0.0182 -0.1108 0.0109 -0.9936>}
 dyn_r_braid_03 {parent=dyn_r_braid_02, loc=<9.4429 0 0>, rot=<0.0244 -0.0545 0.0074 -0.9982>}
 dyn_r_braid_04 {parent=dyn_r_braid_03, loc=<9.0558 0 0>, rot=<-0.0026 -0.0311 -0.0066 -0.9995>}
 dyn_r_braid_05 {parent=dyn_r_braid_04, loc=<9.7886 0 0>, rot=<0.0166 -0.0159 0.0079 -0.9997>}
 dyn_r_braid_06 {parent=dyn_r_braid_05, loc=<9.6982 0 0>, rot=<0.0103 -0.0164 -0.0144 -0.9997>}
 dyn_r_braid_07 {parent=dyn_r_braid_06, loc=<10.1948 0 0>, rot=<-0.0135 0.0198 0.0186 -0.9995>}
 dyn_r_braid_08 {parent=dyn_r_braid_07, loc=<7.54 0 0>, rot=<-0.0424 -0.0214 0.0327 -0.9983>}
 dyn_backhair {parent=head, loc=<3.0391 -3.8436 0>, rot=<0 0 0 -1>}
 dyn_l_hatFlap {parent=head, loc=<8.7405 0.4595 -0.7198>, rot=<-0.0136 -0.0305 0.1531 -0.9876>}
 dyn_r_bang_MSTR {parent=head, loc=<9.7671 8.6611 2.2929>, rot=<0.5765 0.3364 0.6913 0.2768>}
 dyn_r_bang_01 {parent=dyn_r_bang_MSTR, loc=<0 0 0>, rot=<0 0 0 1>}
 dyn_r_bang_02 {parent=dyn_r_bang_01, loc=<4.6848 0 0>, rot=<-0.0744 -0.2345 0.1677 -0.9546>}
 dyn_r_bang_03 {parent=dyn_r_bang_02, loc=<4.3885 0 0>, rot=<0.0146 -0.3392 0.1299 -0.9316>}
 dyn_r_bang_04 {parent=dyn_r_bang_03, loc=<3.5206 0 0>, rot=<0.0975 -0.0636 0.0742 -0.9904>}
 dyn_l_bang_MSTR {parent=head, loc=<8.8949 8.6001 -0.4035>, rot=<-0.436 -0.4896 -0.3283 -0.68>}
 dyn_l_bang_01 {parent=dyn_l_bang_MSTR, loc=<0 0 0>, rot=<0 0 0 1>}
 dyn_l_bang_02 {parent=dyn_l_bang_01, loc=<2.552 0.0427 0.1492>, rot=<-0.0076 -0.3066 0.0891 -0.9476>}
 dyn_l_bang_03 {parent=dyn_l_bang_02, loc=<2.619 0 0>, rot=<0.0541 -0.3046 -0.0762 -0.9479>}
 ponytail_CV_0 {parent=head, loc=<-0.5984 -6.2388 -0.1099>, rot=<0.0029 -0.7071 -0.0029 -0.7071>}
 ponytail_CV_r_0 {parent=head, loc=<-0.5984 -6.2388 6.3598>, rot=<0.0029 -0.7071 -0.0029 -0.7071>}
 ponytail_CV_l_0 {parent=head, loc=<-0.5984 -6.2388 -6.135>, rot=<0.0029 -0.7071 -0.0029 -0.7071>}
 ponytail_CV_1 {parent=spine_05, loc=<0.7957 -13.5621 0.5259>, rot=<0.0284 -0.7065 -0.0284 -0.7065>}
 ponytail_CV_l_1 {parent=spine_05, loc=<0.7194 -12.6133 -5.4993>, rot=<0.0284 -0.7065 -0.0284 -0.7065>}
 ponytail_CV_r_1 {parent=spine_05, loc=<0.7194 -12.6133 6.9955>, rot=<0.0284 -0.7065 -0.0284 -0.7065>}
 thigh_l {parent=pelvis, loc=<-1.0855 -0.2109 -8.7053>, rot=<0.0009 0.0355 0.0155 -0.9993>}
 thigh_r {parent=pelvis, loc=<-1.0853 -0.2109 8.7053>, rot=<0.0355 -0.0009 0.9993 0.0155>}
 ponytail_CV_2 {parent=pelvis, loc=<3.0434 -20.2127 0.0532>, rot=<-0.0279 -0.9991 0.0295 0.0132>}
 ponytail_CV_3 {parent=ponytail_CV_2, loc=<25.5375 0 0>, rot=<0 0 0 -1>}
 ponytail_CV_l_2 {parent=pelvis, loc=<3.0905 -18.8974 -5.972>, rot=<-0.0279 -0.9991 0.0295 0.0132>}
 ponytail_CV_l_3 {parent=ponytail_CV_l_2, loc=<25.5375 0 0>, rot=<0 0 0 -1>}
 ponytail_CV_r_2 {parent=pelvis, loc=<3.0905 -18.8974 6.5228>, rot=<-0.0279 -0.9991 0.0295 0.0132>}
 ponytail_CV_r_3 {parent=ponytail_CV_r_2, loc=<25.5375 0 0>, rot=<0 0 0 -1>}
 ik_hand_root {parent=root, loc=<0 0 0>, rot=<0 0 0 1>}
 ik_hand_gun {parent=ik_hand_root, loc=<-35.2227 3.1613 94.1496>, rot=<-0.5451 -0.2436 -0.5214 0.6097>}
]
