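# Post-processing AnimBlueprint for Quarrel's FaceAcc mesh, animating her two braids with procedural rigs and physics.
# Control Rigs layer skydive, base and crouch hair corrections.
# Braid branches combine per-bone AnimDynamics chains using body sphere limits with bone-constraint setups, including a NoPhysic switch.
# The cached PostSim pose gets crouch, rifle look-up and cape-squash bone offsets, then a RigidBody node with overridden gravity produces the output.
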AnimBlueprint /BRCosmetics/Characters/Player/Female/Medium/Bodies/F_MED_Quarrel/Meshes/Parts/F_MED_Quarrel_FaceAcc_AnimBP
target_skeleton=/Game/Characters/Player/Female/Medium/Bodies/F_MED_Quarrel/Meshes/Parts/F_MED_Quarrel_FaceAcc_Skeleton
graphs [
 AnimGraph {nodes [
   n1 = Control Rig(Alpha=Get skydive_alpha() -> (skydive_alpha:double)) {
    alphaBoolBlend {blendOption=Linear}
    controlRigAssetReference {blueprintRigClass=/Game/Characters/Player/Female/Medium/Bodies/F_MED_Quarrel/Meshes/Parts/F_MED_Quarrel_Skydive_Hair_CtrlRig.F_MED_Quarrel_Skydive_Hair_CtrlRig_C}
   }
   n2 = Control Rig(Source=@n1) {
    alphaBoolBlend {blendOption=Linear}
    controlRigAssetReference {blueprintRigClass=/Game/Characters/Player/Female/Medium/Bodies/F_MED_Quarrel/Meshes/Parts/F_MED_Quarrel_Hair_CtrlRig.F_MED_Quarrel_Hair_CtrlRig_C}
   }
   Control Rig(Source=@n2, Alpha=Get crouch_alpha() -> (crouch_alpha:double)) {
    alphaBoolBlend {blendOption=Linear}
    controlRigAssetReference {blueprintRigClass=/Game/Characters/Player/Female/Medium/Bodies/F_MED_Quarrel/Meshes/Parts/F_MED_Quarrel_Hair_Crouch_CtrlRig.F_MED_Quarrel_Hair_Crouch_CtrlRig_C}
   }
   n3 = "Transform (Modify) Bone - Bone: dyn_r_braid_MSTR"(
    ComponentPose=Local To Component(LocalPose=Use cached pose 'PostSim'())
    Alpha=Get LookUpRifle() -> (LookUpRifle:double)
    Translation="0,-2,2"
    Rotation="0,0,0"
    Scale="1,1,1"
   ) {
    rotationMode=BMM_Additive
    scaleMode=BMM_Additive
    translationMode=BMM_Additive
    alphaBoolBlend {blendInTime=0.1, blendOption=Linear, blendOutTime=0.1}
    boneToModify {boneName=dyn_r_braid_MSTR}
    translation {y=-2, z=2}
   }
   n4 = Component To Local(ComponentPose=@n3)
   n5 = Use cached pose 'PostSim'()
   n6 = Get bIsCrouching()
   n7 = Get isWearingCape() -> (isWearingCape:double)
   n8 = "Transform (Modify) Bone - Bone: dyn_l_braid_MSTR"(
    ComponentPose=Local To Component(LocalPose=Blend Poses by bool(BlendPose_0=@n4, BlendPose_1=@n5, BlendTime_1=0.25, bActiveValue=@n6) {blendType=Linear})
    Alpha=@n7
    Translation="-3.5,0,0"
    Rotation="0.000,0.000,-5.603"
    Scale="0.9,1,1"
   ) {
    rotationMode=BMM_Additive
    scaleMode=BMM_Additive
    translationMode=BMM_Additive
    alphaBoolBlend {blendOption=Linear}
    boneToModify {boneName=dyn_l_braid_MSTR}
    rotation {roll=-5.603}
    scale {x=0.9}
    translation {x=-3.5}
   }
   n9 = "Transform (Modify) Bone - Bone: dyn_l_braid_01"(ComponentPose=@n8, Alpha=@n7, Translation="0,0,0", Rotation="0,0,0", Scale="0.5,1,1") {
    rotationMode=BMM_Additive
    scaleMode=BMM_Additive
    translationMode=BMM_Additive
    alphaBoolBlend {blendOption=Linear}
    boneToModify {boneName=dyn_l_braid_01}
    scale {x=0.5}
   }
   n10 = "Transform (Modify) Bone - Bone: dyn_l_braid_02"(ComponentPose=@n9, Alpha=@n7, Translation="0,0,0", Rotation="0,0,0", Scale="0.8,1,1") {
    rotationMode=BMM_Additive
    scaleMode=BMM_Additive
    translationMode=BMM_Additive
    alphaBoolBlend {blendOption=Linear}
    boneToModify {boneName=dyn_l_braid_02}
    scale {x=0.8}
   }
   n11 = "Transform (Modify) Bone - Bone: dyn_l_braid_02"(ComponentPose=@n10, Alpha=@n7, Translation="0,0,0", Rotation="0,0,0", Scale="0.8,1,1") {
    rotationMode=BMM_Additive
    scaleMode=BMM_Additive
    translationMode=BMM_Additive
    alphaBoolBlend {blendOption=Linear}
    boneToModify {boneName=dyn_l_braid_02}
    scale {x=0.8}
   }
   n12 = "Transform (Modify) Bone - Bone: dyn_l_braid_03"(ComponentPose=@n11, Alpha=@n7, Translation="0,0,0", Rotation="0,0,0", Scale="0.7,1,1") {
    rotationMode=BMM_Additive
    scaleMode=BMM_Additive
    translationMode=BMM_Additive
    alphaBoolBlend {blendOption=Linear}
    boneToModify {boneName=dyn_l_braid_03}
    scale {x=0.7}
   }
   n13 = "Transform (Modify) Bone - Bone: dyn_l_braid_04"(ComponentPose=@n12, Alpha=@n7, Translation="0,0,0", Rotation="0,0,0", Scale="0.6,1,1") {
    rotationMode=BMM_Additive
    scaleMode=BMM_Additive
    translationMode=BMM_Additive
    alphaBoolBlend {blendOption=Linear}
    boneToModify {boneName=dyn_l_braid_04}
    scale {x=0.6}
   }
   n14 = "Transform (Modify) Bone - Bone: dyn_l_braid_05"(ComponentPose=@n13, Alpha=@n7, Translation="0,0,0", Rotation="0,0,0", Scale="1,1,1") {
    rotationMode=BMM_Additive
    scaleMode=BMM_Additive
    translationMode=BMM_Additive
    alphaBoolBlend {blendOption=Linear}
    boneToModify {boneName=dyn_l_braid_05}
   }
   n15 = "Transform (Modify) Bone - Bone: dyn_r_braid_MSTR"(ComponentPose=@n14, Alpha=@n7, Translation="1,0,0", Rotation="0.000,0.000,-3.772", Scale="0.9,1,1") {
    rotationMode=BMM_Additive
    scaleMode=BMM_Additive
    translationMode=BMM_Additive
    alphaBoolBlend {blendOption=Linear}
    boneToModify {boneName=dyn_r_braid_MSTR}
    rotation {roll=-3.772}
    scale {x=0.9}
    translation {x=1}
   }
   n16 = "Transform (Modify) Bone - Bone: dyn_r_braid_01"(ComponentPose=@n15, Alpha=@n7, Translation="0,0,0", Rotation="0,0,0", Scale="0.5,1,1") {
    rotationMode=BMM_Additive
    scaleMode=BMM_Additive
    translationMode=BMM_Additive
    alphaBoolBlend {blendOption=Linear}
    boneToModify {boneName=dyn_r_braid_01}
    scale {x=0.5}
   }
   n17 = "Transform (Modify) Bone - Bone: dyn_r_braid_02"(ComponentPose=@n16, Alpha=@n7, Translation="0,0,0", Rotation="0,0,0", Scale="0.8,1,1") {
    rotationMode=BMM_Additive
    scaleMode=BMM_Additive
    translationMode=BMM_Additive
    alphaBoolBlend {blendOption=Linear}
    boneToModify {boneName=dyn_r_braid_02}
    scale {x=0.8}
   }
   n18 = "Transform (Modify) Bone - Bone: dyn_r_braid_02"(ComponentPose=@n17, Alpha=@n7, Translation="0,0,0", Rotation="0,0,0", Scale="0.8,1,1") {
    rotationMode=BMM_Additive
    scaleMode=BMM_Additive
    translationMode=BMM_Additive
    alphaBoolBlend {blendOption=Linear}
    boneToModify {boneName=dyn_r_braid_02}
    scale {x=0.8}
   }
   n19 = "Transform (Modify) Bone - Bone: dyn_r_braid_03"(ComponentPose=@n18, Alpha=@n7, Translation="0,0,0", Rotation="0,0,0", Scale="0.7,1,1") {
    rotationMode=BMM_Additive
    scaleMode=BMM_Additive
    translationMode=BMM_Additive
    alphaBoolBlend {blendOption=Linear}
    boneToModify {boneName=dyn_r_braid_03}
    scale {x=0.7}
   }
   n20 = "Transform (Modify) Bone - Bone: dyn_r_braid_04"(ComponentPose=@n19, Alpha=@n7, Translation="0,0,0", Rotation="0,0,0", Scale="0.6,1,1") {
    rotationMode=BMM_Additive
    scaleMode=BMM_Additive
    translationMode=BMM_Additive
    alphaBoolBlend {blendOption=Linear}
    boneToModify {boneName=dyn_r_braid_04}
    scale {x=0.6}
   }
   n21 = "Transform (Modify) Bone - Bone: dyn_r_braid_05"(ComponentPose=@n20, Alpha=@n7, Translation="0,0,0", Rotation="0,0,0", Scale="1,1,1") {
    rotationMode=BMM_Additive
    scaleMode=BMM_Additive
    translationMode=BMM_Additive
    alphaBoolBlend {blendOption=Linear}
    boneToModify {boneName=dyn_r_braid_05}
   }
   n22 = RigidBody!(
    ComponentPose=@n21
    bAlphaBoolEnabled:bool
    OverrideWorldGravity:Vector=Get Gravity_Override() -> (Gravity_Override:Vector)
    ExternalForce
    ComponentLinearAccScale:Vector=Get ACC() -> (ACC:Vector)
    ComponentLinearVelScale:Vector=Get VEL() -> (VEL:Vector)
   ) -> (Pose) {
    alphaInputType=Bool
    bClampLinearTranslationLimitToRefPose=true
    bOverrideWorldGravity=true
    overridePhysicsAsset=&/F_MED_Quarrel_FaceAcc_FrontEnd_Physics
    simulationSpace=BaseBoneSpace
    alphaBoolBlend {blendOption=Linear}
    baseBoneRef {boneName=root}
    componentAppliedLinearAccClamp=<2500 2500 2500>
    componentLinearAccScale=<0.3 0.3 0.3>
    componentLinearVelScale=<0.3 0.3 0.3>
    overrideWorldGravity {z=-980}
   }
   Output Pose(Result=Component To Local(ComponentPose=@n22)) {layerGroup=None}
   n23 = "Transform (Modify) Bone - Bone: dyn_l_braid_MSTR"(
    ComponentPose=Local To Component(LocalPose=Use cached pose 'ControlRig'())
    Alpha=Get isCustomEmotePlaying2() -> (isCustomEmotePlaying2:double)
    Translation="3,-5,0"
    Rotation="20,0,0"
    Scale="1,1,1"
   ) {
    rotationMode=BMM_Additive
    rotationSpace=BCS_BoneSpace
    scaleMode=BMM_Additive
    scaleSpace=BCS_BoneSpace
    translationMode=BMM_Additive
    translationSpace=BCS_BoneSpace
    alphaBoolBlend {blendOption=Linear}
    alphaScaleBiasClamp {bInterpResult=true, interpSpeedDecreasing=20, interpSpeedIncreasing=20}
    boneToModify {boneName=dyn_l_braid_MSTR}
    rotation {pitch=20}
    translation=<3 -5>
   }
   n24 = Anim Dynamics - Bone: dyn_l_braid_01!(ComponentPose=@n23, Alpha=1, GravityOverride:Vector="0,0,0", ExternalForce="0,0,0") -> (Pose) {
    angularDampingOverride=1
    bAngularSpring=true
    bGravityOverrideInSimSpace=true
    bOverrideAngularDamping=true
    bUseSphericalLimits=true
    gravityScale=0.5
    simulationSpace=BoneRelative
    alphaBoolBlend {blendOption=Linear}
    boundBone {boneName=dyn_l_braid_01}
    componentAppliedLinearAccClamp=<2000 2000 4000>
    componentLinearAccScale {z=0.5}
    componentLinearVelScale {z=0.5}
    physicsBodyDefinitions [{
      collisionType=CustomSphere
      sphereCollisionRadius=1
      boundBone {boneName=dyn_l_braid_01}
      boxExtents=<15 25 25>
      constraintSetup {angularLimitsMax=<20 45 20>, angularLimitsMin=<-20 -45 -90>, angularTarget=<0 0 0>}
      localJointOffset=<8 0 0>
     }]
    relativeSpaceBone {boneName=root}
    sphericalLimits [
     {limitRadius=65.9176, drivingBone {boneName=spine_05}, sphereLocalOffset=<-12.8628 53 0>}
     {limitRadius=50, drivingBone {boneName=pelvis}, sphereLocalOffset=<-2.6826 40 0>}
     {limitRadius=28, drivingBone {boneName=spine_02}, sphereLocalOffset=<2.6549 15 0>}
     {limitRadius=0, drivingBone {boneName=None}, sphereLocalOffset=<0 0 0>}
    ]
   }
   n25 = Anim Dynamics - Bone: dyn_l_braid_02!(ComponentPose=@n24, Alpha=1, GravityOverride:Vector="0,0,0", ExternalForce="0,0,0") -> (Pose) {
    angularDampingOverride=1
    bAngularSpring=true
    bGravityOverrideInSimSpace=true
    bOverrideAngularDamping=true
    bUseSphericalLimits=true
    gravityScale=0.5
    simulationSpace=BoneRelative
    alphaBoolBlend {blendOption=Linear}
    alphaScaleBiasClamp {scale=0.5}
    boundBone {boneName=dyn_l_braid_02}
    componentAppliedLinearAccClamp=<2000 2000 4000>
    componentLinearAccScale {z=0.5}
    componentLinearVelScale {z=0.5}
    physicsBodyDefinitions [{
      collisionType=CustomSphere
      sphereCollisionRadius=1
      boundBone {boneName=dyn_l_braid_02}
      boxExtents=<15 25 25>
      constraintSetup {angularLimitsMax=<20 45 20>, angularLimitsMin=<-20 -45 -90>, angularTarget=<0 0 0>}
      localJointOffset=<8 0 0>
     }]
    relativeSpaceBone {boneName=root}
    sphericalLimits [
     {limitRadius=65.9176, drivingBone {boneName=spine_05}, sphereLocalOffset=<-12.8628 53 0>}
     {limitRadius=50, drivingBone {boneName=pelvis}, sphereLocalOffset=<-2.6826 40 0>}
     {limitRadius=28, drivingBone {boneName=spine_02}, sphereLocalOffset=<2.6549 15 0>}
    ]
   }
   n26 = Anim Dynamics - Bone: dyn_l_braid_03!(ComponentPose=@n25, Alpha=1, GravityOverride:Vector="0,0,0", ExternalForce="0,0,0") -> (Pose) {
    angularDampingOverride=1
    bAngularSpring=true
    bGravityOverrideInSimSpace=true
    bOverrideAngularDamping=true
    bUseSphericalLimits=true
    gravityScale=0.5
    simulationSpace=BoneRelative
    alphaBoolBlend {blendOption=Linear}
    alphaScaleBiasClamp {scale=0.35}
    boundBone {boneName=dyn_l_braid_03}
    componentAppliedLinearAccClamp=<2000 2000 4000>
    componentLinearAccScale {z=0.5}
    componentLinearVelScale {z=0.5}
    physicsBodyDefinitions [{
      collisionType=CustomSphere
      sphereCollisionRadius=1
      boundBone {boneName=dyn_l_braid_03}
      boxExtents=<15 25 25>
      constraintSetup {angularLimitsMax=<20 45 20>, angularLimitsMin=<-20 -45 -90>, angularTarget=<0 0 0>}
      localJointOffset=<8 0 0>
     }]
    relativeSpaceBone {boneName=root}
    sphericalLimits [
     {limitRadius=65.9176, drivingBone {boneName=spine_05}, sphereLocalOffset=<-12.8628 52 0>}
     {limitRadius=50, drivingBone {boneName=pelvis}, sphereLocalOffset=<-2.6826 38.079 0>}
     {limitRadius=28, drivingBone {boneName=spine_02}, sphereLocalOffset=<2.6549 18 0>}
    ]
   }
   n27 = Anim Dynamics - Bone: dyn_l_braid_04!(ComponentPose=@n26, Alpha=1, GravityOverride:Vector="0,0,0", ExternalForce="0,0,0") -> (Pose) {
    angularDampingOverride=0.9999
    bAngularSpring=true
    bGravityOverrideInSimSpace=true
    bOverrideAngularDamping=true
    bUseSphericalLimits=true
    gravityScale=0.5
    simulationSpace=BoneRelative
    alphaBoolBlend {blendOption=Linear}
    alphaScaleBiasClamp {scale=0.35}
    boundBone {boneName=dyn_l_braid_04}
    componentAppliedLinearAccClamp=<2000 2000 8000>
    componentLinearAccScale {z=0.5}
    componentLinearVelScale {z=0.5}
    physicsBodyDefinitions [{
      collisionType=CustomSphere
      sphereCollisionRadius=1
      boundBone {boneName=dyn_l_braid_04}
      boxExtents=<12 22 22>
      constraintSetup {angularLimitsMax=<20 45 90>, angularLimitsMin=<-20 -45 -90>, angularTarget=<0 0 0>}
      localJointOffset=<10 0 0>
     }]
    relativeSpaceBone {boneName=root}
    sphericalLimits [
     {limitRadius=65.9176, drivingBone {boneName=spine_05}, sphereLocalOffset=<-12.8628 52 0>}
     {limitRadius=50, drivingBone {boneName=pelvis}, sphereLocalOffset=<-2.6826 38.079 0>}
     {limitRadius=28, drivingBone {boneName=spine_02}, sphereLocalOffset=<2.6549 15 0>}
     {limitRadius=1, drivingBone {boneName=pelvis}, sphereLocalOffset=<-30 0 0>}
    ]
   }
   n28 = Anim Dynamics - Bone: dyn_l_braid_05!(ComponentPose=@n27, Alpha=1, GravityOverride:Vector="0,0,0", ExternalForce="0,0,0") -> (Pose) {
    angularDampingOverride=0.9999
    bAngularSpring=true
    bGravityOverrideInSimSpace=true
    bOverrideAngularDamping=true
    bUseSphericalLimits=true
    gravityScale=0.5
    simulationSpace=BoneRelative
    alphaBoolBlend {blendOption=Linear}
    alphaScaleBiasClamp {scale=0.35}
    boundBone {boneName=dyn_l_braid_05}
    componentAppliedLinearAccClamp=<2000 2000 8000>
    componentLinearAccScale {z=0.5}
    componentLinearVelScale {z=0.5}
    physicsBodyDefinitions [{
      collisionType=CustomSphere
      sphereCollisionRadius=1
      boundBone {boneName=dyn_l_braid_05}
      boxExtents=<12 22 22>
      constraintSetup {angularLimitsMax=<20 45 90>, angularLimitsMin=<-20 -45 -90>, angularTarget=<0 0 0>}
      localJointOffset=<10 0 0>
     }]
    relativeSpaceBone {boneName=root}
    sphericalLimits [
     {limitRadius=65.9176, drivingBone {boneName=spine_05}, sphereLocalOffset=<-12.8628 52 0>}
     {limitRadius=50, drivingBone {boneName=pelvis}, sphereLocalOffset=<-2.6826 38.079 0>}
     {limitRadius=28, drivingBone {boneName=spine_02}, sphereLocalOffset=<2.6549 15 0>}
     {limitRadius=1, drivingBone {boneName=pelvis}, sphereLocalOffset=<-30 0 0>}
    ]
   }
   n29 = Anim Dynamics - Bone: dyn_l_braid_06!(ComponentPose=@n28, Alpha=1, GravityOverride:Vector="0,0,0", ExternalForce="0,0,0") -> (Pose) {
    angularDampingOverride=0.9999
    bAngularSpring=true
    bGravityOverrideInSimSpace=true
    bOverrideAngularDamping=true
    bUseSphericalLimits=true
    gravityScale=0.5
    simulationSpace=BoneRelative
    alphaBoolBlend {blendOption=Linear}
    alphaScaleBiasClamp {scale=0.35}
    boundBone {boneName=dyn_l_braid_06}
    componentAppliedLinearAccClamp=<2000 2000 8000>
    componentLinearAccScale {z=0.5}
    componentLinearVelScale {z=0.5}
    physicsBodyDefinitions [{
      collisionType=CustomSphere
      sphereCollisionRadius=1
      boundBone {boneName=dyn_l_braid_06}
      boxExtents=<12 22 22>
      constraintSetup {angularLimitsMax=<20 45 90>, angularLimitsMin=<-20 -45 -90>, angularTarget=<0 0 0>}
      localJointOffset=<10 0 0>
     }]
    relativeSpaceBone {boneName=root}
    sphericalLimits [
     {limitRadius=65.9176, drivingBone {boneName=spine_05}, sphereLocalOffset=<-12.8628 52 0>}
     {limitRadius=50, drivingBone {boneName=pelvis}, sphereLocalOffset=<-2.6826 38.079 0>}
     {limitRadius=28, drivingBone {boneName=spine_02}, sphereLocalOffset=<2.6549 15 0>}
     {limitRadius=1, drivingBone {boneName=pelvis}, sphereLocalOffset=<-30 0 0>}
    ]
   }
   n30 = Anim Dynamics - Bone: dyn_l_braid_07!(ComponentPose=@n29, Alpha=1, GravityOverride:Vector="0,0,0", ExternalForce="0,0,0") -> (Pose) {
    angularDampingOverride=0.9999
    bAngularSpring=true
    bGravityOverrideInSimSpace=true
    bOverrideAngularDamping=true
    bUseSphericalLimits=true
    gravityScale=0.5
    simulationSpace=BoneRelative
    alphaBoolBlend {blendOption=Linear}
    alphaScaleBiasClamp {scale=0.4}
    boundBone {boneName=dyn_l_braid_07}
    componentAppliedLinearAccClamp=<2000 2000 8000>
    componentLinearAccScale {z=0.5}
    componentLinearVelScale {z=0.5}
    physicsBodyDefinitions [{
      collisionType=CustomSphere
      sphereCollisionRadius=1
      boundBone {boneName=dyn_l_braid_07}
      boxExtents=<12 22 22>
      constraintSetup {angularLimitsMax=<20 90 90>, angularLimitsMin=<-20 -90 -90>, angularTarget=<0 0 0>}
      localJointOffset=<10 0 0>
     }]
    relativeSpaceBone {boneName=root}
    sphericalLimits [
     {limitRadius=65.9176, drivingBone {boneName=spine_05}, sphereLocalOffset=<-12.8628 62 0>}
     {limitRadius=50, drivingBone {boneName=pelvis}, sphereLocalOffset=<-2.6826 38.079 0>}
     {limitRadius=28, drivingBone {boneName=spine_02}, sphereLocalOffset=<2.6549 15 0>}
     {limitRadius=1, drivingBone {boneName=pelvis}, sphereLocalOffset=<-30 0 0>}
    ]
   }
   n31 = Anim Dynamics - Bone: dyn_l_braid_08!(ComponentPose=@n30, Alpha=1, GravityOverride:Vector="0,0,0", ExternalForce="0,0,0") -> (Pose) {
    angularDampingOverride=0.9999
    bAngularSpring=true
    bGravityOverrideInSimSpace=true
    bOverrideAngularDamping=true
    bUseSphericalLimits=true
    gravityScale=0.5
    simulationSpace=BoneRelative
    alphaBoolBlend {blendOption=Linear}
    alphaScaleBiasClamp {scale=0.5}
    boundBone {boneName=dyn_l_braid_08}
    componentAppliedLinearAccClamp=<2000 2000 8000>
    componentLinearAccScale {z=0.5}
    componentLinearVelScale {z=0.5}
    physicsBodyDefinitions [{
      collisionType=CustomSphere
      sphereCollisionRadius=1
      boundBone {boneName=dyn_l_braid_08}
      boxExtents=<12 22 22>
      constraintSetup {angularLimitsMax=<20 90 90>, angularLimitsMin=<-20 -90 -90>, angularTarget=<0 0 0>}
      localJointOffset=<10 0 0>
     }]
    relativeSpaceBone {boneName=root}
    sphericalLimits [
     {limitRadius=65.9176, drivingBone {boneName=spine_05}, sphereLocalOffset=<-12.8628 62 0>}
     {limitRadius=50, drivingBone {boneName=pelvis}, sphereLocalOffset=<-2.6826 38.079 0>}
     {limitRadius=28, drivingBone {boneName=spine_02}, sphereLocalOffset=<2.6549 15 0>}
     {limitRadius=1, drivingBone {boneName=pelvis}, sphereLocalOffset=<-30 0 0>}
    ]
   }
   n32 = Anim Dynamics - Bone: dyn_r_braid_01!(ComponentPose=@n31, Alpha=1, GravityOverride:Vector="0,0,0", ExternalForce="0,0,0") -> (Pose) {
    angularDampingOverride=1
    bAngularSpring=true
    bGravityOverrideInSimSpace=true
    bOverrideAngularDamping=true
    bUseSphericalLimits=true
    gravityScale=0.47
    simulationSpace=BoneRelative
    alphaBoolBlend {blendOption=Linear}
    boundBone {boneName=dyn_r_braid_01}
    componentAppliedLinearAccClamp=<2000 2000 4000>
    componentLinearAccScale=<0.1 0.1 0.55>
    componentLinearVelScale=<0.1 0.1 0.55>
    physicsBodyDefinitions [{
      collisionType=CustomSphere
      sphereCollisionRadius=1
      boundBone {boneName=dyn_r_braid_01}
      boxExtents=<15 25 25>
      constraintSetup {angularLimitsMax=<20 45 10>, angularLimitsMin=<-20 -45 -10>, angularTarget=<0 0 0>}
      localJointOffset=<13 0 0>
     }]
    relativeSpaceBone {boneName=root}
    sphericalLimits [
     {limitRadius=65.9176, drivingBone {boneName=spine_05}, sphereLocalOffset=<-12.8628 53 0>}
     {limitRadius=50, drivingBone {boneName=pelvis}, sphereLocalOffset=<-2.6826 40 0>}
     {limitRadius=28, drivingBone {boneName=spine_02}, sphereLocalOffset=<2.6549 15 0>}
     {limitRadius=0, drivingBone {boneName=None}, sphereLocalOffset=<0 0 0>}
    ]
   }
   n33 = Anim Dynamics - Bone: dyn_r_braid_02!(ComponentPose=@n32, Alpha=1, GravityOverride:Vector="0,0,0", ExternalForce="0,0,0") -> (Pose) {
    angularDampingOverride=1
    bAngularSpring=true
    bGravityOverrideInSimSpace=true
    bOverrideAngularDamping=true
    bUseSphericalLimits=true
    gravityScale=0.47
    simulationSpace=BoneRelative
    alphaBoolBlend {blendOption=Linear}
    alphaScaleBiasClamp {scale=0.5}
    boundBone {boneName=dyn_r_braid_02}
    componentAppliedLinearAccClamp=<2000 2000 4000>
    componentLinearAccScale=<0.1 0.1 0.55>
    componentLinearVelScale=<0.1 0.1 0.55>
    physicsBodyDefinitions [{
      collisionType=CustomSphere
      sphereCollisionRadius=1
      boundBone {boneName=dyn_r_braid_02}
      boxExtents=<15 25 25>
      constraintSetup {angularLimitsMax=<20 45 10>, angularLimitsMin=<-20 -45 -10>, angularTarget=<0 0 0>}
      localJointOffset=<13 0 0>
     }]
    relativeSpaceBone {boneName=root}
    sphericalLimits [
     {limitRadius=65.9176, drivingBone {boneName=spine_05}, sphereLocalOffset=<-12.8628 53 0>}
     {limitRadius=50, drivingBone {boneName=pelvis}, sphereLocalOffset=<-2.6826 40 0>}
     {limitRadius=28, drivingBone {boneName=spine_02}, sphereLocalOffset=<2.6549 15 0>}
    ]
   }
   n34 = Anim Dynamics - Bone: dyn_r_braid_03!(ComponentPose=@n33, Alpha=1, GravityOverride:Vector="0,0,0", ExternalForce="0,0,0") -> (Pose) {
    angularDampingOverride=1
    bAngularSpring=true
    bGravityOverrideInSimSpace=true
    bOverrideAngularDamping=true
    bUseSphericalLimits=true
    gravityScale=0.47
    simulationSpace=BoneRelative
    alphaBoolBlend {blendOption=Linear}
    alphaScaleBiasClamp {scale=0.35}
    boundBone {boneName=dyn_r_braid_03}
    componentAppliedLinearAccClamp=<2000 2000 4000>
    componentLinearAccScale=<0.1 0.1 0.55>
    componentLinearVelScale=<0.1 0.1 0.55>
    physicsBodyDefinitions [{
      collisionType=CustomSphere
      sphereCollisionRadius=1
      boundBone {boneName=dyn_r_braid_03}
      boxExtents=<15 25 25>
      constraintSetup {angularLimitsMax=<20 45 10>, angularLimitsMin=<-20 -45 -10>, angularTarget=<0 0 0>}
      localJointOffset=<13 0 0>
     }]
    relativeSpaceBone {boneName=root}
    sphericalLimits [
     {limitRadius=65.9176, drivingBone {boneName=spine_05}, sphereLocalOffset=<-12.8628 52 0>}
     {limitRadius=50, drivingBone {boneName=pelvis}, sphereLocalOffset=<-2.6826 38.079 0>}
     {limitRadius=28, drivingBone {boneName=spine_02}, sphereLocalOffset=<2.6549 18 0>}
    ]
   }
   n35 = Anim Dynamics - Bone: dyn_r_braid_04!(ComponentPose=@n34, Alpha=1, GravityOverride:Vector="0,0,0", ExternalForce="0,0,0") -> (Pose) {
    angularDampingOverride=0.9999
    bAngularSpring=true
    bGravityOverrideInSimSpace=true
    bOverrideAngularDamping=true
    bUseSphericalLimits=true
    gravityScale=0.47
    simulationSpace=BoneRelative
    alphaBoolBlend {blendOption=Linear}
    alphaScaleBiasClamp {scale=0.35}
    boundBone {boneName=dyn_r_braid_04}
    componentAppliedLinearAccClamp=<2000 2000 8000>
    componentLinearAccScale=<0.1 0.1 0.55>
    componentLinearVelScale=<0.1 0.1 0.55>
    physicsBodyDefinitions [{
      collisionType=CustomSphere
      sphereCollisionRadius=1
      boundBone {boneName=dyn_r_braid_04}
      boxExtents=<15 25 25>
      constraintSetup {angularLimitsMax=<20 45 90>, angularLimitsMin=<-20 -45 -90>, angularTarget=<0 0 0>}
      localJointOffset=<13 0 0>
     }]
    relativeSpaceBone {boneName=root}
    sphericalLimits [
     {limitRadius=65.9176, drivingBone {boneName=spine_05}, sphereLocalOffset=<-12.8628 52 0>}
     {limitRadius=50, drivingBone {boneName=pelvis}, sphereLocalOffset=<-2.6826 38.079 0>}
     {limitRadius=28, drivingBone {boneName=spine_02}, sphereLocalOffset=<2.6549 15 0>}
     {limitRadius=1, drivingBone {boneName=pelvis}, sphereLocalOffset=<-30 0 0>}
    ]
   }
   n36 = Anim Dynamics - Bone: dyn_r_braid_05!(ComponentPose=@n35, Alpha=1, GravityOverride:Vector="0,0,0", ExternalForce="0,0,0") -> (Pose) {
    angularDampingOverride=0.9999
    bAngularSpring=true
    bGravityOverrideInSimSpace=true
    bOverrideAngularDamping=true
    bUseSphericalLimits=true
    gravityScale=0.47
    simulationSpace=BoneRelative
    alphaBoolBlend {blendOption=Linear}
    alphaScaleBiasClamp {scale=0.35}
    boundBone {boneName=dyn_r_braid_05}
    componentAppliedLinearAccClamp=<2000 2000 8000>
    componentLinearAccScale=<0.1 0.1 0.55>
    componentLinearVelScale=<0.1 0.1 0.55>
    physicsBodyDefinitions [{
      collisionType=CustomSphere
      sphereCollisionRadius=1
      boundBone {boneName=dyn_r_braid_05}
      boxExtents=<15 25 25>
      constraintSetup {angularLimitsMax=<20 45 90>, angularLimitsMin=<-20 -45 -90>, angularTarget=<0 0 0>}
      localJointOffset=<13 0 0>
     }]
    relativeSpaceBone {boneName=root}
    sphericalLimits [
     {limitRadius=65.9176, drivingBone {boneName=spine_05}, sphereLocalOffset=<-12.8628 52 0>}
     {limitRadius=50, drivingBone {boneName=pelvis}, sphereLocalOffset=<-2.6826 38.079 0>}
     {limitRadius=28, drivingBone {boneName=spine_02}, sphereLocalOffset=<2.6549 15 0>}
     {limitRadius=1, drivingBone {boneName=pelvis}, sphereLocalOffset=<-30 0 0>}
    ]
   }
   n37 = Anim Dynamics - Bone: dyn_r_braid_06!(ComponentPose=@n36, Alpha=1, GravityOverride:Vector="0,0,0", ExternalForce="0,0,0") -> (Pose) {
    angularDampingOverride=0.9999
    bAngularSpring=true
    bGravityOverrideInSimSpace=true
    bOverrideAngularDamping=true
    bUseSphericalLimits=true
    gravityScale=0.47
    simulationSpace=BoneRelative
    alphaBoolBlend {blendOption=Linear}
    alphaScaleBiasClamp {scale=0.35}
    boundBone {boneName=dyn_r_braid_06}
    componentAppliedLinearAccClamp=<2000 2000 8000>
    componentLinearAccScale=<0.1 0.1 0.55>
    componentLinearVelScale=<0.1 0.1 0.55>
    physicsBodyDefinitions [{
      collisionType=CustomSphere
      sphereCollisionRadius=1
      boundBone {boneName=dyn_r_braid_06}
      boxExtents=<15 25 25>
      constraintSetup {angularLimitsMax=<20 45 90>, angularLimitsMin=<-20 -45 -90>, angularTarget=<0 0 0>}
      localJointOffset=<13 0 0>
     }]
    relativeSpaceBone {boneName=root}
    sphericalLimits [
     {limitRadius=65.9176, drivingBone {boneName=spine_05}, sphereLocalOffset=<-12.8628 52 0>}
     {limitRadius=50, drivingBone {boneName=pelvis}, sphereLocalOffset=<-2.6826 38.079 0>}
     {limitRadius=28, drivingBone {boneName=spine_02}, sphereLocalOffset=<2.6549 15 0>}
     {limitRadius=1, drivingBone {boneName=pelvis}, sphereLocalOffset=<-30 0 0>}
    ]
   }
   n38 = Anim Dynamics - Bone: dyn_r_braid_07!(ComponentPose=@n37, Alpha=1, GravityOverride:Vector="0,0,0", ExternalForce="0,0,0") -> (Pose) {
    angularDampingOverride=0.9999
    bAngularSpring=true
    bGravityOverrideInSimSpace=true
    bOverrideAngularDamping=true
    bUseSphericalLimits=true
    gravityScale=0.47
    simulationSpace=BoneRelative
    alphaBoolBlend {blendOption=Linear}
    alphaScaleBiasClamp {scale=0.4}
    boundBone {boneName=dyn_r_braid_07}
    componentAppliedLinearAccClamp=<2000 2000 8000>
    componentLinearAccScale=<0.1 0.1 0.55>
    componentLinearVelScale=<0.1 0.1 0.55>
    physicsBodyDefinitions [{
      collisionType=CustomSphere
      sphereCollisionRadius=1
      boundBone {boneName=dyn_r_braid_07}
      boxExtents=<15 25 25>
      constraintSetup {angularLimitsMax=<20 90 90>, angularLimitsMin=<-20 -90 -90>, angularTarget=<0 0 0>}
      localJointOffset=<13 0 0>
     }]
    relativeSpaceBone {boneName=root}
    sphericalLimits [
     {limitRadius=65.9176, drivingBone {boneName=spine_05}, sphereLocalOffset=<-12.8628 62 0>}
     {limitRadius=50, drivingBone {boneName=pelvis}, sphereLocalOffset=<-2.6826 38.079 0>}
     {limitRadius=28, drivingBone {boneName=spine_02}, sphereLocalOffset=<2.6549 15 0>}
     {limitRadius=1, drivingBone {boneName=pelvis}, sphereLocalOffset=<-30 0 0>}
    ]
   }
   n39 = Anim Dynamics - Bone: dyn_r_braid_08!(ComponentPose=@n38, Alpha=1, GravityOverride:Vector="0,0,0", ExternalForce="0,0,0") -> (Pose) {
    angularDampingOverride=0.9999
    bAngularSpring=true
    bGravityOverrideInSimSpace=true
    bOverrideAngularDamping=true
    bUseSphericalLimits=true
    gravityScale=0.47
    simulationSpace=BoneRelative
    alphaBoolBlend {blendOption=Linear}
    alphaScaleBiasClamp {scale=0.5}
    boundBone {boneName=dyn_r_braid_08}
    componentAppliedLinearAccClamp=<2000 2000 8000>
    componentLinearAccScale=<0.1 0.1 0.55>
    componentLinearVelScale=<0.1 0.1 0.55>
    physicsBodyDefinitions [{
      collisionType=CustomSphere
      sphereCollisionRadius=1
      boundBone {boneName=dyn_r_braid_08}
      boxExtents=<15 25 25>
      constraintSetup {angularLimitsMax=<20 90 90>, angularLimitsMin=<-20 -90 -90>, angularTarget=<0 0 0>}
      localJointOffset=<13 0 0>
     }]
    relativeSpaceBone {boneName=root}
    sphericalLimits [
     {limitRadius=65.9176, drivingBone {boneName=spine_05}, sphereLocalOffset=<-12.8628 62 0>}
     {limitRadius=50, drivingBone {boneName=pelvis}, sphereLocalOffset=<-2.6826 38.079 0>}
     {limitRadius=28, drivingBone {boneName=spine_02}, sphereLocalOffset=<2.6549 15 0>}
     {limitRadius=1, drivingBone {boneName=pelvis}, sphereLocalOffset=<-30 0 0>}
    ]
   }
   n40 = "Constraint - dyn_l_braid_01 (P)"(ComponentPose=Local To Component(LocalPose=Use cached pose 'ControlRig'()), Alpha=1, ConstraintWeights_0=1) {
    alphaBoolBlend {blendOption=Linear}
    boneToModify {boneName=dyn_l_braid_01}
    constraintSetup [{targetBone {boneName=neck_01}}]
    constraintWeights [1]
   }
   n41 = "Constraint - dyn_r_braid_01 (P)"(ComponentPose=@n40, Alpha=1, ConstraintWeights_0=1) {
    alphaBoolBlend {blendOption=Linear}
    boneToModify {boneName=dyn_r_braid_01}
    constraintSetup [{targetBone {boneName=neck_01}}]
    constraintWeights [1]
   }
   n42 = "Transform (Modify) Bone - Bone: dyn_l_braid_01"!(ComponentPose=@n41, Alpha=1, Translation="0,0,0", Rotation="0,-15,20") -> (Pose) {
    rotationMode=BMM_Additive
    rotationSpace=BCS_BoneSpace
    scaleMode=BMM_Additive
    translationMode=BMM_Additive
    translationSpace=BCS_BoneSpace
    alphaBoolBlend {blendOption=Linear}
    boneToModify {boneName=dyn_l_braid_01}
    rotation {roll=20, yaw=-15}
   }
   n43 = "Transform (Modify) Bone - Bone: dyn_r_braid_01"!(ComponentPose=@n42, Alpha=1, Translation="0,0,0", Rotation="0,10,20") -> (Pose) {
    rotationMode=BMM_Additive
    rotationSpace=BCS_BoneSpace
    scaleMode=BMM_Additive
    translationMode=BMM_Additive
    translationSpace=BCS_BoneSpace
    alphaBoolBlend {blendOption=Linear}
    boneToModify {boneName=dyn_r_braid_01}
    rotation {roll=20, yaw=10}
   }
   n44 = "Constraint - dyn_l_braid_02 (P)"(ComponentPose=@n43, Alpha=1, ConstraintWeights_0=1) {
    alphaBoolBlend {blendOption=Linear}
    boneToModify {boneName=dyn_l_braid_02}
    constraintSetup [{targetBone {boneName=spine_05}}]
    constraintWeights [1]
   }
   n45 = "Constraint - dyn_r_braid_02 (P)"(ComponentPose=@n44, Alpha=1, ConstraintWeights_0=1) {
    alphaBoolBlend {blendOption=Linear}
    boneToModify {boneName=dyn_r_braid_02}
    constraintSetup [{targetBone {boneName=spine_05}}]
    constraintWeights [1]
   }
   n46 = "Transform (Modify) Bone - Bone: dyn_l_braid_02"!(ComponentPose=@n45, Alpha=1, Translation="1,-3,0", Rotation="5,-10,0") -> (Pose) {
    rotationMode=BMM_Additive
    rotationSpace=BCS_BoneSpace
    scaleMode=BMM_Additive
    translationMode=BMM_Additive
    translationSpace=BCS_BoneSpace
    alphaBoolBlend {blendOption=Linear}
    boneToModify {boneName=dyn_l_braid_02}
    rotation {pitch=5, yaw=-10}
    translation=<1 -3>
   }
   n47 = "Transform (Modify) Bone - Bone: dyn_r_braid_02"!(ComponentPose=@n46, Alpha=1, Translation="1,3,0", Rotation="5,15,0") -> (Pose) {
    rotationMode=BMM_Additive
    rotationSpace=BCS_BoneSpace
    scaleMode=BMM_Additive
    translationMode=BMM_Additive
    translationSpace=BCS_BoneSpace
    alphaBoolBlend {blendOption=Linear}
    boneToModify {boneName=dyn_r_braid_02}
    rotation {pitch=5, yaw=15}
    translation=<1 3>
   }
   n48 = Blend Poses by bool(
    BlendPose_0=Component To Local(ComponentPose=@n47)
    BlendPose_1=Component To Local(ComponentPose=@n39)
    bActiveValue=Get NoPhysic() -> (NoPhysic:bool)
   ) {blendType=Linear}
   n49 = "Constraint - dyn_l_braid_MSTR (P)"(ComponentPose=Local To Component(LocalPose=Use cached pose 'ControlRig'()), Alpha=1, ConstraintWeights_0=1) {
    alphaBoolBlend {blendOption=Linear}
    boneToModify {boneName=dyn_l_braid_MSTR}
    constraintSetup [{targetBone {boneName=head}}]
    constraintWeights [1]
   }
   n50 = "Constraint - dyn_l_braid_01 (PP)"!(ComponentPose=@n49, Alpha=1, ConstraintWeights_0=1, ConstraintWeights_1:float=0) -> (Pose) {
    alphaBoolBlend {blendOption=Linear}
    boneToModify {boneName=dyn_l_braid_01}
    constraintSetup [{targetBone {boneName=spine_05}}, {targetBone {boneName=head}}]
    constraintWeights [1,1]
   }
   n51 = "Constraint - dyn_l_braid_02 (PP)"!(ComponentPose=@n50, Alpha=1, ConstraintWeights_0=1, ConstraintWeights_1:float=0) -> (Pose) {
    alphaBoolBlend {blendOption=Linear}
    boneToModify {boneName=dyn_l_braid_02}
    constraintSetup [{targetBone {boneName=spine_05}}, {targetBone {boneName=head}}]
    constraintWeights [1,1]
   }
   n52 = "Constraint - dyn_r_braid_MSTR (P)"(ComponentPose=@n51, Alpha=1, ConstraintWeights_0=1) {
    alphaBoolBlend {blendOption=Linear}
    boneToModify {boneName=dyn_r_braid_MSTR}
    constraintSetup [{targetBone {boneName=head}}]
    constraintWeights [1]
   }
   n53 = "Constraint - dyn_r_braid_01 (PP)"!(ComponentPose=@n52, Alpha=1, ConstraintWeights_0=0.8, ConstraintWeights_1:float=0.2) -> (Pose) {
    alphaBoolBlend {blendOption=Linear}
    boneToModify {boneName=dyn_r_braid_01}
    constraintSetup [{targetBone {boneName=spine_05}}, {targetBone {boneName=head}}]
    constraintWeights [1,1]
   }
   n54 = "Constraint - dyn_r_braid_02 (PP)"!(ComponentPose=@n53, Alpha=1, ConstraintWeights_0=0.8, ConstraintWeights_1:float=0.2) -> (Pose) {
    alphaBoolBlend {blendOption=Linear}
    boneToModify {boneName=dyn_r_braid_02}
    constraintSetup [{targetBone {boneName=spine_05}}, {targetBone {boneName=head}}]
    constraintWeights [1,1]
   }
   n55 = "Transform (Modify) Bone - Bone: dyn_l_braid_MSTR"!(ComponentPose=@n54, bAlphaBoolEnabled:bool=NOT Boolean(A=Get bIsSkyDiving()), Translation="0,0,0", Rotation="0,-8,0", Scale="1,1,1") -> (Pose) {
    alphaInputType=Bool
    rotationMode=BMM_Additive
    rotationSpace=BCS_BoneSpace
    scaleMode=BMM_Additive
    scaleSpace=BCS_BoneSpace
    translationMode=BMM_Additive
    translationSpace=BCS_BoneSpace
    alphaBoolBlend {blendInTime=0.1, blendOption=Linear, blendOutTime=0.1}
    boneToModify {boneName=dyn_l_braid_MSTR}
    rotation {yaw=-8}
   }
   n56 = "Transform (Modify) Bone - Bone: dyn_r_braid_MSTR"!(ComponentPose=@n55, bAlphaBoolEnabled:bool=True, Translation="0,0,0", Rotation="0,8,0", Scale="1,1,1") -> (Pose) {
    alphaInputType=Bool
    rotationMode=BMM_Additive
    rotationSpace=BCS_BoneSpace
    scaleMode=BMM_Additive
    scaleSpace=BCS_BoneSpace
    translationMode=BMM_Additive
    translationSpace=BCS_BoneSpace
    alphaBoolBlend {blendInTime=0.1, blendOption=Linear, blendOutTime=0.1}
    boneToModify {boneName=dyn_r_braid_MSTR}
    rotation {yaw=8}
   }
   n57 = Anim Dynamics - Bone: dyn_l_braid_04(ComponentPose=@n56, Alpha=1, GravityOverride="0,0,0", ExternalForce="0,0,0") {
    angularDampingOverride=1
    angularSpringConstant=2500
    bAngularSpring=true
    bGravityOverrideInSimSpace=true
    bOverrideAngularDamping=true
    bUseSphericalLimits=true
    gravityScale=0.5
    simulationSpace=RootRelative
    alphaBoolBlend {blendOption=Linear}
    alphaScaleBiasClamp {scale=0.35}
    boundBone {boneName=dyn_l_braid_04}
    componentAppliedLinearAccClamp=<5000 5000 5000>
    componentLinearAccScale {z=0.7}
    componentLinearVelScale {z=0.7}
    physicsBodyDefinitions [{
      collisionType=CustomSphere
      sphereCollisionRadius=1
      boundBone {boneName=dyn_l_braid_04}
      boxExtents=<12 22 22>
      constraintSetup {angularLimitsMax=<20 45 90>, angularLimitsMin=<-20 -45 -90>, angularTarget=<0 0 0>}
      localJointOffset=<10 0 0>
     }]
    sphericalLimits [
     {limitRadius=65.9176, drivingBone {boneName=spine_05}, sphereLocalOffset=<-12.8628 55 0>}
     {limitRadius=50, drivingBone {boneName=pelvis}, sphereLocalOffset=<-2.6826 40 0>}
     {limitRadius=28, drivingBone {boneName=spine_02}, sphereLocalOffset=<2.6549 15 0>}
     {limitRadius=1, drivingBone {boneName=pelvis}, sphereLocalOffset=<-30 0 0>}
    ]
   }
   n58 = Anim Dynamics - Bone: dyn_l_braid_05(ComponentPose=@n57, Alpha=1, GravityOverride="0,0,0", ExternalForce="0,0,0") {
    angularDampingOverride=1
    angularSpringConstant=2500
    bAngularSpring=true
    bGravityOverrideInSimSpace=true
    bOverrideAngularDamping=true
    bUseSphericalLimits=true
    gravityScale=0.5
    simulationSpace=RootRelative
    alphaBoolBlend {blendOption=Linear}
    alphaScaleBiasClamp {scale=0.35}
    boundBone {boneName=dyn_l_braid_05}
    componentAppliedLinearAccClamp=<5000 5000 5000>
    componentLinearAccScale {z=0.7}
    componentLinearVelScale {z=0.7}
    physicsBodyDefinitions [{
      collisionType=CustomSphere
      sphereCollisionRadius=1
      boundBone {boneName=dyn_l_braid_05}
      boxExtents=<12 22 22>
      constraintSetup {angularLimitsMax=<20 45 90>, angularLimitsMin=<-20 -45 -90>, angularTarget=<0 0 0>}
      localJointOffset=<10 0 0>
     }]
    sphericalLimits [
     {limitRadius=65.9176, drivingBone {boneName=spine_05}, sphereLocalOffset=<-12.8628 55 0>}
     {limitRadius=50, drivingBone {boneName=pelvis}, sphereLocalOffset=<-2.6826 40 0>}
     {limitRadius=28, drivingBone {boneName=spine_02}, sphereLocalOffset=<2.6549 15 0>}
     {limitRadius=1, drivingBone {boneName=pelvis}, sphereLocalOffset=<-30 0 0>}
    ]
   }
   n59 = Anim Dynamics - Bone: dyn_l_braid_06(ComponentPose=@n58, Alpha=1, GravityOverride="0,0,0", ExternalForce="0,0,0") {
    angularDampingOverride=1
    angularSpringConstant=2500
    bAngularSpring=true
    bGravityOverrideInSimSpace=true
    bOverrideAngularDamping=true
    bUseSphericalLimits=true
    gravityScale=0.5
    simulationSpace=RootRelative
    alphaBoolBlend {blendOption=Linear}
    alphaScaleBiasClamp {scale=0.35}
    boundBone {boneName=dyn_l_braid_06}
    componentAppliedLinearAccClamp=<5000 5000 5000>
    componentLinearAccScale {z=0.7}
    componentLinearVelScale {z=0.7}
    physicsBodyDefinitions [{
      collisionType=CustomSphere
      sphereCollisionRadius=1
      boundBone {boneName=dyn_l_braid_06}
      boxExtents=<12 22 22>
      constraintSetup {angularLimitsMax=<20 45 90>, angularLimitsMin=<-20 -45 -90>, angularTarget=<0 0 0>}
      localJointOffset=<10 0 0>
     }]
    sphericalLimits [
     {limitRadius=65.9176, drivingBone {boneName=spine_05}, sphereLocalOffset=<-12.8628 55 0>}
     {limitRadius=50, drivingBone {boneName=pelvis}, sphereLocalOffset=<-2.6826 40 0>}
     {limitRadius=28, drivingBone {boneName=spine_02}, sphereLocalOffset=<2.6549 15 0>}
     {limitRadius=1, drivingBone {boneName=pelvis}, sphereLocalOffset=<-30 0 0>}
    ]
   }
   n60 = Anim Dynamics - Bone: dyn_l_braid_07(ComponentPose=@n59, Alpha=1, GravityOverride="0,0,0", ExternalForce="0,0,0") {
    angularDampingOverride=1
    angularSpringConstant=2500
    bAngularSpring=true
    bGravityOverrideInSimSpace=true
    bOverrideAngularDamping=true
    bUseSphericalLimits=true
    gravityScale=0.5
    simulationSpace=RootRelative
    alphaBoolBlend {blendOption=Linear}
    alphaScaleBiasClamp {scale=0.4}
    boundBone {boneName=dyn_l_braid_07}
    componentAppliedLinearAccClamp=<5000 5000 5000>
    componentLinearAccScale {z=0.7}
    componentLinearVelScale {z=0.7}
    physicsBodyDefinitions [{
      collisionType=CustomSphere
      sphereCollisionRadius=1
      boundBone {boneName=dyn_l_braid_07}
      boxExtents=<12 22 22>
      constraintSetup {angularLimitsMax=<20 90 90>, angularLimitsMin=<-20 -90 -90>, angularTarget=<0 0 0>}
      localJointOffset=<10 0 0>
     }]
    sphericalLimits [
     {limitRadius=65.9176, drivingBone {boneName=spine_05}, sphereLocalOffset=<-12.8628 65 0>}
     {limitRadius=50, drivingBone {boneName=pelvis}, sphereLocalOffset=<-2.6826 40 0>}
     {limitRadius=28, drivingBone {boneName=spine_02}, sphereLocalOffset=<2.6549 17 0>}
     {limitRadius=1, drivingBone {boneName=pelvis}, sphereLocalOffset=<-30 0 0>}
    ]
   }
   n61 = Anim Dynamics - Bone: dyn_l_braid_08(ComponentPose=@n60, Alpha=1, GravityOverride="0,0,0", ExternalForce="0,0,0") {
    angularDampingOverride=1
    angularSpringConstant=2500
    bAngularSpring=true
    bGravityOverrideInSimSpace=true
    bOverrideAngularDamping=true
    bUseSphericalLimits=true
    gravityScale=0.5
    simulationSpace=RootRelative
    alphaBoolBlend {blendOption=Linear}
    alphaScaleBiasClamp {scale=0.5}
    boundBone {boneName=dyn_l_braid_08}
    componentAppliedLinearAccClamp=<5000 5000 5000>
    componentLinearAccScale {z=0.7}
    componentLinearVelScale {z=0.7}
    physicsBodyDefinitions [{
      collisionType=CustomSphere
      sphereCollisionRadius=1
      boundBone {boneName=dyn_l_braid_08}
      boxExtents=<12 22 22>
      constraintSetup {angularLimitsMax=<20 90 90>, angularLimitsMin=<-20 -90 -90>, angularTarget=<0 0 0>}
      localJointOffset=<10 0 0>
     }]
    sphericalLimits [
     {limitRadius=65.9176, drivingBone {boneName=spine_05}, sphereLocalOffset=<-12.8628 65 0>}
     {limitRadius=50, drivingBone {boneName=pelvis}, sphereLocalOffset=<-2.6826 40 0>}
     {limitRadius=28, drivingBone {boneName=spine_02}, sphereLocalOffset=<2.6549 17 0>}
     {limitRadius=1, drivingBone {boneName=pelvis}, sphereLocalOffset=<-30 0 0>}
    ]
   }
   n62 = Anim Dynamics - Bone: dyn_r_braid_04(ComponentPose=@n61, Alpha=1, GravityOverride="0,0,0", ExternalForce="0,0,0") {
    angularDampingOverride=1
    angularSpringConstant=2500
    bAngularSpring=true
    bGravityOverrideInSimSpace=true
    bOverrideAngularDamping=true
    bUseSphericalLimits=true
    gravityScale=0.5
    simulationSpace=RootRelative
    alphaBoolBlend {blendOption=Linear}
    alphaScaleBiasClamp {scale=0.35}
    boundBone {boneName=dyn_r_braid_04}
    componentAppliedLinearAccClamp=<5000 5000 5000>
    componentLinearAccScale {z=0.7}
    componentLinearVelScale {z=0.7}
    physicsBodyDefinitions [{
      collisionType=CustomSphere
      sphereCollisionRadius=1
      boundBone {boneName=dyn_r_braid_04}
      boxExtents=<14 28 28>
      constraintSetup {angularLimitsMax=<20 45 90>, angularLimitsMin=<-20 -45 -90>, angularTarget=<0 0 0>}
      localJointOffset=<10 -0.5 0>
     }]
    sphericalLimits [
     {limitRadius=65.9176, drivingBone {boneName=spine_05}, sphereLocalOffset=<-12.8628 55 0>}
     {limitRadius=50, drivingBone {boneName=pelvis}, sphereLocalOffset=<-2.6826 40 0>}
     {limitRadius=28, drivingBone {boneName=spine_02}, sphereLocalOffset=<2.6549 15 0>}
     {limitRadius=1, drivingBone {boneName=pelvis}, sphereLocalOffset=<-30 0 0>}
    ]
   }
   n63 = Anim Dynamics - Bone: dyn_r_braid_05(ComponentPose=@n62, Alpha=1, GravityOverride="0,0,0", ExternalForce="0,0,0") {
    angularDampingOverride=1
    angularSpringConstant=2500
    bAngularSpring=true
    bGravityOverrideInSimSpace=true
    bOverrideAngularDamping=true
    bUseSphericalLimits=true
    gravityScale=0.5
    simulationSpace=RootRelative
    alphaBoolBlend {blendOption=Linear}
    alphaScaleBiasClamp {scale=0.35}
    boundBone {boneName=dyn_r_braid_05}
    componentAppliedLinearAccClamp=<5000 5000 5000>
    componentLinearAccScale {z=0.7}
    componentLinearVelScale {z=0.7}
    physicsBodyDefinitions [{
      collisionType=CustomSphere
      sphereCollisionRadius=1
      boundBone {boneName=dyn_r_braid_05}
      boxExtents=<14 28 28>
      constraintSetup {angularLimitsMax=<20 45 90>, angularLimitsMin=<-20 -45 -90>, angularTarget=<0 0 0>}
      localJointOffset=<10 -0.5 0>
     }]
    sphericalLimits [
     {limitRadius=65.9176, drivingBone {boneName=spine_05}, sphereLocalOffset=<-12.8628 55 0>}
     {limitRadius=50, drivingBone {boneName=pelvis}, sphereLocalOffset=<-2.6826 40 0>}
     {limitRadius=28, drivingBone {boneName=spine_02}, sphereLocalOffset=<2.6549 15 0>}
     {limitRadius=1, drivingBone {boneName=pelvis}, sphereLocalOffset=<-30 0 0>}
    ]
   }
   n64 = Anim Dynamics - Bone: dyn_r_braid_06(ComponentPose=@n63, Alpha=1, GravityOverride="0,0,0", ExternalForce="0,0,0") {
    angularDampingOverride=1
    angularSpringConstant=2500
    bAngularSpring=true
    bGravityOverrideInSimSpace=true
    bOverrideAngularDamping=true
    bUseSphericalLimits=true
    gravityScale=0.5
    simulationSpace=RootRelative
    alphaBoolBlend {blendOption=Linear}
    alphaScaleBiasClamp {scale=0.35}
    boundBone {boneName=dyn_r_braid_06}
    componentAppliedLinearAccClamp=<5000 5000 5000>
    componentLinearAccScale {z=0.7}
    componentLinearVelScale {z=0.7}
    physicsBodyDefinitions [{
      collisionType=CustomSphere
      sphereCollisionRadius=1
      boundBone {boneName=dyn_r_braid_06}
      boxExtents=<14 28 28>
      constraintSetup {angularLimitsMax=<20 45 90>, angularLimitsMin=<-20 -45 -90>, angularTarget=<0 0 0>}
      localJointOffset=<10 -0.5 0>
     }]
    sphericalLimits [
     {limitRadius=65.9176, drivingBone {boneName=spine_05}, sphereLocalOffset=<-12.8628 55 0>}
     {limitRadius=50, drivingBone {boneName=pelvis}, sphereLocalOffset=<-2.6826 40 0>}
     {limitRadius=28, drivingBone {boneName=spine_02}, sphereLocalOffset=<2.6549 15 0>}
     {limitRadius=1, drivingBone {boneName=pelvis}, sphereLocalOffset=<-30 0 0>}
    ]
   }
   n65 = Anim Dynamics - Bone: dyn_r_braid_07(ComponentPose=@n64, Alpha=1, GravityOverride="0,0,0", ExternalForce="0,0,0") {
    angularDampingOverride=1
    angularSpringConstant=2500
    bAngularSpring=true
    bGravityOverrideInSimSpace=true
    bOverrideAngularDamping=true
    bUseSphericalLimits=true
    gravityScale=0.5
    simulationSpace=RootRelative
    alphaBoolBlend {blendOption=Linear}
    alphaScaleBiasClamp {scale=0.4}
    boundBone {boneName=dyn_r_braid_07}
    componentAppliedLinearAccClamp=<5000 5000 5000>
    componentLinearAccScale {z=0.7}
    componentLinearVelScale {z=0.7}
    physicsBodyDefinitions [{
      collisionType=CustomSphere
      sphereCollisionRadius=1
      boundBone {boneName=dyn_r_braid_07}
      boxExtents=<14 28 28>
      constraintSetup {angularLimitsMax=<20 90 90>, angularLimitsMin=<-20 -90 -90>, angularTarget=<0 0 0>}
      localJointOffset=<10 -0.5 0>
     }]
    sphericalLimits [
     {limitRadius=65.9176, drivingBone {boneName=spine_05}, sphereLocalOffset=<-12.8628 65 0>}
     {limitRadius=50, drivingBone {boneName=pelvis}, sphereLocalOffset=<-2.6826 40 0>}
     {limitRadius=28, drivingBone {boneName=spine_02}, sphereLocalOffset=<2.6549 17 0>}
     {limitRadius=1, drivingBone {boneName=pelvis}, sphereLocalOffset=<-30 0 0>}
    ]
   }
   n66 = Anim Dynamics - Bone: dyn_r_braid_08(ComponentPose=@n65, Alpha=1, GravityOverride="0,0,0", ExternalForce="0,0,0") {
    angularDampingOverride=1
    angularSpringConstant=2500
    bAngularSpring=true
    bGravityOverrideInSimSpace=true
    bOverrideAngularDamping=true
    bUseSphericalLimits=true
    gravityScale=0.5
    simulationSpace=RootRelative
    alphaBoolBlend {blendOption=Linear}
    alphaScaleBiasClamp {scale=0.5}
    boundBone {boneName=dyn_r_braid_08}
    componentAppliedLinearAccClamp=<5000 5000 5000>
    componentLinearAccScale {z=0.7}
    componentLinearVelScale {z=0.7}
    physicsBodyDefinitions [{
      collisionType=CustomSphere
      sphereCollisionRadius=1
      boundBone {boneName=dyn_r_braid_08}
      boxExtents=<14 28 28>
      constraintSetup {angularLimitsMax=<20 90 90>, angularLimitsMin=<-20 -90 -90>, angularTarget=<0 0 0>}
      localJointOffset=<10 -0.5 0>
     }]
    sphericalLimits [
     {limitRadius=65.9176, drivingBone {boneName=spine_05}, sphereLocalOffset=<-12.8628 65 0>}
     {limitRadius=50, drivingBone {boneName=pelvis}, sphereLocalOffset=<-2.6826 40 0>}
     {limitRadius=28, drivingBone {boneName=spine_02}, sphereLocalOffset=<2.6549 17 0>}
     {limitRadius=1, drivingBone {boneName=pelvis}, sphereLocalOffset=<-30 0 0>}
    ]
   }
   n67 = Component To Local(ComponentPose=@n66)
   n68 = Get isWearingBackbling()
   n69 = "Constraint - dyn_l_braid_MSTR (P)"(ComponentPose=Local To Component(LocalPose=Use cached pose 'ControlRig'()), Alpha=1, ConstraintWeights_0=1) {
    alphaBoolBlend {blendOption=Linear}
    boneToModify {boneName=dyn_l_braid_MSTR}
    constraintSetup [{targetBone {boneName=spine_05}}]
    constraintWeights [1]
   }
   n70 = "Constraint - dyn_l_braid_01 (P)"(ComponentPose=@n69, Alpha=1, ConstraintWeights_0=1) {
    alphaBoolBlend {blendOption=Linear}
    boneToModify {boneName=dyn_l_braid_01}
    constraintSetup [{targetBone {boneName=spine_05}}]
    constraintWeights [1]
   }
   n71 = "Constraint - dyn_l_braid_02 (P)"(ComponentPose=@n70, Alpha=1, ConstraintWeights_0=1) {
    alphaBoolBlend {blendOption=Linear}
    boneToModify {boneName=dyn_l_braid_02}
    constraintSetup [{targetBone {boneName=spine_05}}]
    constraintWeights [1]
   }
   n72 = "Constraint - dyn_l_braid_03 (P)"(ComponentPose=@n71, Alpha=1, ConstraintWeights_0=1) {
    alphaBoolBlend {blendOption=Linear}
    boneToModify {boneName=dyn_l_braid_03}
    constraintSetup [{targetBone {boneName=spine_05}}]
    constraintWeights [1]
   }
   n73 = "Constraint - dyn_r_braid_MSTR (P)"(ComponentPose=@n72, Alpha=1, ConstraintWeights_0=1) {
    alphaBoolBlend {blendOption=Linear}
    boneToModify {boneName=dyn_r_braid_MSTR}
    constraintSetup [{targetBone {boneName=spine_05}}]
    constraintWeights [1]
   }
   n74 = "Constraint - dyn_r_braid_01 (P)"(ComponentPose=@n73, Alpha=1, ConstraintWeights_0=1) {
    alphaBoolBlend {blendOption=Linear}
    boneToModify {boneName=dyn_r_braid_01}
    constraintSetup [{targetBone {boneName=spine_05}}]
    constraintWeights [1]
   }
   n75 = "Constraint - dyn_r_braid_02 (P)"(ComponentPose=@n74, Alpha=1, ConstraintWeights_0=1) {
    alphaBoolBlend {blendOption=Linear}
    boneToModify {boneName=dyn_r_braid_02}
    constraintSetup [{targetBone {boneName=spine_05}}]
    constraintWeights [1]
   }
   n76 = "Constraint - dyn_r_braid_03 (P)"(ComponentPose=@n75, Alpha=1, ConstraintWeights_0=1) {
    alphaBoolBlend {blendOption=Linear}
    boneToModify {boneName=dyn_r_braid_03}
    constraintSetup [{targetBone {boneName=spine_05}}]
    constraintWeights [1]
   }
   n77 = "Transform (Modify) Bone - Bone: dyn_l_braid_MSTR"!(ComponentPose=@n76, Alpha:float=1, Translation="-4,0,0", Rotation="0,0,0", Scale="1,1,1") -> (Pose) {
    rotationMode=BMM_Additive
    scaleMode=BMM_Additive
    translationMode=BMM_Additive
    alphaBoolBlend {blendOption=Linear}
    boneToModify {boneName=dyn_l_braid_MSTR}
    translation {x=-4}
   }
   n78 = "Transform (Modify) Bone - Bone: dyn_r_braid_MSTR"!(ComponentPose=@n77, Alpha:float=1, Translation="4,0,0", Rotation="0,0,0", Scale="1,1,1") -> (Pose) {
    rotationMode=BMM_Additive
    scaleMode=BMM_Additive
    translationMode=BMM_Additive
    alphaBoolBlend {blendOption=Linear}
    boneToModify {boneName=dyn_r_braid_MSTR}
    translation {x=4}
   }
   n79 = Anim Dynamics - Bone: dyn_l_braid_04(ComponentPose=@n78, Alpha=1, GravityOverride="0,0,0", ExternalForce="0,0,0") {
    angularDampingOverride=1
    angularSpringConstant=2500
    bAngularSpring=true
    bGravityOverrideInSimSpace=true
    bOverrideAngularDamping=true
    bUseGravityOverride=true
    bUseSphericalLimits=true
    gravityScale=0.35
    simulationSpace=RootRelative
    alphaBoolBlend {blendOption=Linear}
    alphaScaleBiasClamp {scale=0.35}
    boundBone {boneName=dyn_l_braid_04}
    componentAppliedLinearAccClamp=<5000 5000 5000>
    componentLinearAccScale {z=0.44}
    componentLinearVelScale {z=0.44}
    physicsBodyDefinitions [{
      collisionType=CustomSphere
      sphereCollisionRadius=1
      boundBone {boneName=dyn_l_braid_04}
      boxExtents=<12 22 22>
      constraintSetup {angularLimitsMax=<20 20 20>, angularLimitsMin=<-20 -20 -20>, angularTarget=<0 0 0>}
      localJointOffset=<10 0 0>
     }]
    sphericalLimits [
     {limitRadius=65.9176, drivingBone {boneName=spine_05}, sphereLocalOffset=<-12.8628 52 0>}
     {limitRadius=50, drivingBone {boneName=pelvis}, sphereLocalOffset=<-2.6826 38.079 0>}
     {limitRadius=28, drivingBone {boneName=spine_02}, sphereLocalOffset=<2.6549 15 0>}
     {limitRadius=1, drivingBone {boneName=pelvis}, sphereLocalOffset=<-30 0 0>}
    ]
   }
   n80 = Anim Dynamics - Bone: dyn_l_braid_05(ComponentPose=@n79, Alpha=1, GravityOverride="0,0,0", ExternalForce="0,0,0") {
    angularDampingOverride=1
    angularSpringConstant=2500
    bAngularSpring=true
    bGravityOverrideInSimSpace=true
    bOverrideAngularDamping=true
    bUseGravityOverride=true
    bUseSphericalLimits=true
    gravityScale=0.35
    simulationSpace=RootRelative
    alphaBoolBlend {blendOption=Linear}
    alphaScaleBiasClamp {scale=0.35}
    boundBone {boneName=dyn_l_braid_05}
    componentAppliedLinearAccClamp=<5000 5000 5000>
    componentLinearAccScale {z=0.44}
    componentLinearVelScale {z=0.44}
    physicsBodyDefinitions [{
      collisionType=CustomSphere
      sphereCollisionRadius=1
      boundBone {boneName=dyn_l_braid_05}
      boxExtents=<12 22 22>
      constraintSetup {angularLimitsMax=<45 45 45>, angularLimitsMin=<-45 -45 -45>, angularTarget=<0 0 0>}
      localJointOffset=<10 0 0>
     }]
    sphericalLimits [
     {limitRadius=65.9176, drivingBone {boneName=spine_05}, sphereLocalOffset=<-12.8628 52 0>}
     {limitRadius=50, drivingBone {boneName=pelvis}, sphereLocalOffset=<-2.6826 38.079 0>}
     {limitRadius=28, drivingBone {boneName=spine_02}, sphereLocalOffset=<2.6549 15 0>}
     {limitRadius=1, drivingBone {boneName=pelvis}, sphereLocalOffset=<-30 0 0>}
    ]
   }
   n81 = Anim Dynamics - Bone: dyn_l_braid_06(ComponentPose=@n80, Alpha=1, GravityOverride="0,0,0", ExternalForce="0,0,0") {
    angularDampingOverride=1
    angularSpringConstant=2500
    bAngularSpring=true
    bGravityOverrideInSimSpace=true
    bOverrideAngularDamping=true
    bUseGravityOverride=true
    bUseSphericalLimits=true
    gravityScale=0.35
    simulationSpace=RootRelative
    alphaBoolBlend {blendOption=Linear}
    alphaScaleBiasClamp {scale=0.35}
    boundBone {boneName=dyn_l_braid_06}
    componentAppliedLinearAccClamp=<5000 5000 5000>
    componentLinearAccScale {z=0.44}
    componentLinearVelScale {z=0.44}
    physicsBodyDefinitions [{
      collisionType=CustomSphere
      sphereCollisionRadius=1
      boundBone {boneName=dyn_l_braid_06}
      boxExtents=<12 22 22>
      constraintSetup {angularLimitsMax=<45 45 45>, angularLimitsMin=<-45 -45 -45>, angularTarget=<0 0 0>}
      localJointOffset=<10 0 0>
     }]
    sphericalLimits [
     {limitRadius=65.9176, drivingBone {boneName=spine_05}, sphereLocalOffset=<-12.8628 52 0>}
     {limitRadius=50, drivingBone {boneName=pelvis}, sphereLocalOffset=<-2.6826 38.079 0>}
     {limitRadius=28, drivingBone {boneName=spine_02}, sphereLocalOffset=<2.6549 15 0>}
     {limitRadius=1, drivingBone {boneName=pelvis}, sphereLocalOffset=<-30 0 0>}
    ]
   }
   n82 = Anim Dynamics - Bone: dyn_l_braid_07(ComponentPose=@n81, Alpha=1, GravityOverride="0,0,0", ExternalForce="0,0,0") {
    angularDampingOverride=1
    angularSpringConstant=2500
    bAngularSpring=true
    bGravityOverrideInSimSpace=true
    bOverrideAngularDamping=true
    bUseGravityOverride=true
    bUseSphericalLimits=true
    gravityScale=0.35
    simulationSpace=RootRelative
    alphaBoolBlend {blendOption=Linear}
    alphaScaleBiasClamp {scale=0.4}
    boundBone {boneName=dyn_l_braid_07}
    componentAppliedLinearAccClamp=<5000 5000 5000>
    componentLinearAccScale {z=0.46}
    componentLinearVelScale {z=0.46}
    physicsBodyDefinitions [{
      collisionType=CustomSphere
      sphereCollisionRadius=1
      boundBone {boneName=dyn_l_braid_07}
      boxExtents=<12 22 22>
      constraintSetup {angularLimitsMax=<45 45 45>, angularLimitsMin=<-45 -45 -45>, angularTarget=<0 0 0>}
      localJointOffset=<10 0 0>
     }]
    sphericalLimits [
     {limitRadius=65.9176, drivingBone {boneName=spine_05}, sphereLocalOffset=<-12.8628 62 0>}
     {limitRadius=50, drivingBone {boneName=pelvis}, sphereLocalOffset=<-2.6826 38.079 0>}
     {limitRadius=28, drivingBone {boneName=spine_02}, sphereLocalOffset=<2.6549 15 0>}
     {limitRadius=1, drivingBone {boneName=pelvis}, sphereLocalOffset=<-30 0 0>}
    ]
   }
   n83 = Anim Dynamics - Bone: dyn_l_braid_08(ComponentPose=@n82, Alpha=1, GravityOverride="0,0,0", ExternalForce="0,0,0") {
    angularDampingOverride=1
    angularSpringConstant=2500
    bAngularSpring=true
    bGravityOverrideInSimSpace=true
    bOverrideAngularDamping=true
    bUseGravityOverride=true
    bUseSphericalLimits=true
    gravityScale=0.35
    simulationSpace=RootRelative
    alphaBoolBlend {blendOption=Linear}
    alphaScaleBiasClamp {scale=0.5}
    boundBone {boneName=dyn_l_braid_08}
    componentAppliedLinearAccClamp=<5000 5000 5000>
    componentLinearAccScale {z=0.46}
    componentLinearVelScale {z=0.46}
    physicsBodyDefinitions [{
      collisionType=CustomSphere
      sphereCollisionRadius=1
      boundBone {boneName=dyn_l_braid_08}
      boxExtents=<12 22 22>
      constraintSetup {angularLimitsMax=<45 45 45>, angularLimitsMin=<-45 -45 -45>, angularTarget=<0 0 0>}
      localJointOffset=<10 0 0>
     }]
    sphericalLimits [
     {limitRadius=65.9176, drivingBone {boneName=spine_05}, sphereLocalOffset=<-12.8628 62 0>}
     {limitRadius=50, drivingBone {boneName=pelvis}, sphereLocalOffset=<-2.6826 38.079 0>}
     {limitRadius=28, drivingBone {boneName=spine_02}, sphereLocalOffset=<2.6549 15 0>}
     {limitRadius=1, drivingBone {boneName=pelvis}, sphereLocalOffset=<-30 0 0>}
    ]
   }
   n84 = Anim Dynamics - Bone: dyn_r_braid_04(ComponentPose=@n83, Alpha=1, GravityOverride="0,0,0", ExternalForce="0,0,0") {
    angularDampingOverride=1
    angularSpringConstant=2500
    bAngularSpring=true
    bGravityOverrideInSimSpace=true
    bOverrideAngularDamping=true
    bUseGravityOverride=true
    bUseSphericalLimits=true
    gravityScale=0.35
    simulationSpace=RootRelative
    alphaBoolBlend {blendOption=Linear}
    alphaScaleBiasClamp {scale=0.35}
    boundBone {boneName=dyn_r_braid_04}
    componentAppliedLinearAccClamp=<5000 5000 5000>
    componentLinearAccScale {z=0.44}
    componentLinearVelScale {z=0.44}
    physicsBodyDefinitions [{
      collisionType=CustomSphere
      sphereCollisionRadius=1
      boundBone {boneName=dyn_r_braid_04}
      boxExtents=<12 22 22>
      constraintSetup {angularLimitsMax=<45 45 45>, angularLimitsMin=<-45 -45 -45>, angularTarget=<0 0 0>}
      localJointOffset=<10 0 0>
     }]
    sphericalLimits [
     {limitRadius=65.9176, drivingBone {boneName=spine_05}, sphereLocalOffset=<-12.8628 52 0>}
     {limitRadius=50, drivingBone {boneName=pelvis}, sphereLocalOffset=<-2.6826 38.079 0>}
     {limitRadius=28, drivingBone {boneName=spine_02}, sphereLocalOffset=<2.6549 15 0>}
     {limitRadius=1, drivingBone {boneName=pelvis}, sphereLocalOffset=<-30 0 0>}
    ]
   }
   n85 = Anim Dynamics - Bone: dyn_r_braid_05(ComponentPose=@n84, Alpha=1, GravityOverride="0,0,0", ExternalForce="0,0,0") {
    angularDampingOverride=1
    angularSpringConstant=2500
    bAngularSpring=true
    bGravityOverrideInSimSpace=true
    bOverrideAngularDamping=true
    bUseGravityOverride=true
    bUseSphericalLimits=true
    gravityScale=0.35
    simulationSpace=RootRelative
    alphaBoolBlend {blendOption=Linear}
    alphaScaleBiasClamp {scale=0.35}
    boundBone {boneName=dyn_r_braid_05}
    componentAppliedLinearAccClamp=<5000 5000 5000>
    componentLinearAccScale {z=0.44}
    componentLinearVelScale {z=0.44}
    physicsBodyDefinitions [{
      collisionType=CustomSphere
      sphereCollisionRadius=1
      boundBone {boneName=dyn_r_braid_05}
      boxExtents=<12 22 22>
      constraintSetup {angularLimitsMax=<45 45 45>, angularLimitsMin=<-45 -45 -45>, angularTarget=<0 0 0>}
      localJointOffset=<10 0 0>
     }]
    sphericalLimits [
     {limitRadius=65.9176, drivingBone {boneName=spine_05}, sphereLocalOffset=<-12.8628 52 0>}
     {limitRadius=50, drivingBone {boneName=pelvis}, sphereLocalOffset=<-2.6826 38.079 0>}
     {limitRadius=28, drivingBone {boneName=spine_02}, sphereLocalOffset=<2.6549 15 0>}
     {limitRadius=1, drivingBone {boneName=pelvis}, sphereLocalOffset=<-30 0 0>}
    ]
   }
   n86 = Anim Dynamics - Bone: dyn_r_braid_06(ComponentPose=@n85, Alpha=1, GravityOverride="0,0,0", ExternalForce="0,0,0") {
    angularDampingOverride=1
    angularSpringConstant=2500
    bAngularSpring=true
    bGravityOverrideInSimSpace=true
    bOverrideAngularDamping=true
    bUseGravityOverride=true
    bUseSphericalLimits=true
    gravityScale=0.35
    simulationSpace=RootRelative
    alphaBoolBlend {blendOption=Linear}
    alphaScaleBiasClamp {scale=0.35}
    boundBone {boneName=dyn_r_braid_06}
    componentAppliedLinearAccClamp=<5000 5000 5000>
    componentLinearAccScale {z=0.44}
    componentLinearVelScale {z=0.44}
    physicsBodyDefinitions [{
      collisionType=CustomSphere
      sphereCollisionRadius=1
      boundBone {boneName=dyn_r_braid_06}
      boxExtents=<12 22 22>
      constraintSetup {angularLimitsMax=<45 45 45>, angularLimitsMin=<-45 -45 -45>, angularTarget=<0 0 0>}
      localJointOffset=<10 0 0>
     }]
    sphericalLimits [
     {limitRadius=65.9176, drivingBone {boneName=spine_05}, sphereLocalOffset=<-12.8628 52 0>}
     {limitRadius=50, drivingBone {boneName=pelvis}, sphereLocalOffset=<-2.6826 38.079 0>}
     {limitRadius=28, drivingBone {boneName=spine_02}, sphereLocalOffset=<2.6549 15 0>}
     {limitRadius=1, drivingBone {boneName=pelvis}, sphereLocalOffset=<-30 0 0>}
    ]
   }
   n87 = Anim Dynamics - Bone: dyn_r_braid_07(ComponentPose=@n86, Alpha=1, GravityOverride="0,0,0", ExternalForce="0,0,0") {
    angularDampingOverride=1
    angularSpringConstant=2500
    bAngularSpring=true
    bGravityOverrideInSimSpace=true
    bOverrideAngularDamping=true
    bUseGravityOverride=true
    bUseSphericalLimits=true
    gravityScale=0.35
    simulationSpace=RootRelative
    alphaBoolBlend {blendOption=Linear}
    alphaScaleBiasClamp {scale=0.4}
    boundBone {boneName=dyn_r_braid_07}
    componentAppliedLinearAccClamp=<5000 5000 5000>
    componentLinearAccScale {z=0.46}
    componentLinearVelScale {z=0.46}
    physicsBodyDefinitions [{
      collisionType=CustomSphere
      sphereCollisionRadius=1
      boundBone {boneName=dyn_r_braid_07}
      boxExtents=<12 22 22>
      constraintSetup {angularLimitsMax=<45 45 45>, angularLimitsMin=<-45 -45 -45>, angularTarget=<0 0 0>}
      localJointOffset=<10 0 0>
     }]
    sphericalLimits [
     {limitRadius=65.9176, drivingBone {boneName=spine_05}, sphereLocalOffset=<-12.8628 62 0>}
     {limitRadius=50, drivingBone {boneName=pelvis}, sphereLocalOffset=<-2.6826 38.079 0>}
     {limitRadius=28, drivingBone {boneName=spine_02}, sphereLocalOffset=<2.6549 15 0>}
     {limitRadius=1, drivingBone {boneName=pelvis}, sphereLocalOffset=<-30 0 0>}
    ]
   }
   n88 = Anim Dynamics - Bone: dyn_r_braid_08(ComponentPose=@n87, Alpha=1, GravityOverride="0,0,0", ExternalForce="0,0,0") {
    angularDampingOverride=1
    angularSpringConstant=2500
    bAngularSpring=true
    bGravityOverrideInSimSpace=true
    bOverrideAngularDamping=true
    bUseGravityOverride=true
    bUseSphericalLimits=true
    gravityScale=0.35
    simulationSpace=RootRelative
    alphaBoolBlend {blendOption=Linear}
    alphaScaleBiasClamp {scale=0.5}
    boundBone {boneName=dyn_r_braid_08}
    componentAppliedLinearAccClamp=<5000 5000 5000>
    componentLinearAccScale {z=0.46}
    componentLinearVelScale {z=0.46}
    physicsBodyDefinitions [{
      collisionType=CustomSphere
      sphereCollisionRadius=1
      boundBone {boneName=dyn_r_braid_08}
      boxExtents=<12 22 22>
      constraintSetup {angularLimitsMax=<45 45 45>, angularLimitsMin=<-45 -45 -45>, angularTarget=<0 0 0>}
      localJointOffset=<10 0 0>
     }]
    sphericalLimits [
     {limitRadius=65.9176, drivingBone {boneName=spine_05}, sphereLocalOffset=<-12.8628 62 0>}
     {limitRadius=50, drivingBone {boneName=pelvis}, sphereLocalOffset=<-2.6826 38.079 0>}
     {limitRadius=28, drivingBone {boneName=spine_02}, sphereLocalOffset=<2.6549 15 0>}
     {limitRadius=1, drivingBone {boneName=pelvis}, sphereLocalOffset=<-30 0 0>}
    ]
   }
   n89 = Get isNotWearingBackbling() -> (isNotWearingBackbling:double)
   n90 = Anim Dynamics - Bone: dyn_l_braid_01(ComponentPose=Local To Component(LocalPose=Use cached pose 'ControlRig'()), Alpha=@n89, GravityOverride="0,0,0", ExternalForce="0,0,0") {
    angularDampingOverride=1
    angularSpringConstant=2500
    bAngularSpring=true
    bGravityOverrideInSimSpace=true
    bOverrideAngularDamping=true
    bUseGravityOverride=true
    bUseSphericalLimits=true
    gravityScale=0.35
    simulationSpace=RootRelative
    alphaBoolBlend {blendOption=Linear}
    boundBone {boneName=dyn_l_braid_01}
    componentAppliedLinearAccClamp=<200 200 200>
    componentLinearAccScale {z=0.44}
    componentLinearVelScale {z=0.44}
    physicsBodyDefinitions [{
      collisionType=CustomSphere
      sphereCollisionRadius=1
      boundBone {boneName=dyn_l_braid_01}
      boxExtents=<22 30 30>
      constraintSetup {angularLimitsMax=<45 0 45>, angularLimitsMin=<-45 -45 -45>, angularTarget=<0 0 0>}
      localJointOffset=<10 -5 0>
     }]
    sphericalLimits [
     {limitRadius=65.9176, drivingBone {boneName=spine_05}, sphereLocalOffset=<-12.8628 53 0>}
     {limitRadius=50, drivingBone {boneName=pelvis}, sphereLocalOffset=<-2.6826 40 0>}
     {limitRadius=28, drivingBone {boneName=spine_02}, sphereLocalOffset=<2.6549 15 0>}
    ]
   }
   n91 = Anim Dynamics - Bone: dyn_l_braid_02(ComponentPose=@n90, Alpha=@n89, GravityOverride="0,0,0", ExternalForce="0,0,0") {
    angularDampingOverride=1
    angularSpringConstant=2500
    bAngularSpring=true
    bGravityOverrideInSimSpace=true
    bOverrideAngularDamping=true
    bUseGravityOverride=true
    bUseSphericalLimits=true
    gravityScale=0.35
    simulationSpace=RootRelative
    alphaBoolBlend {blendOption=Linear}
    alphaScaleBiasClamp {scale=0.5}
    boundBone {boneName=dyn_l_braid_02}
    componentAppliedLinearAccClamp=<200 200 200>
    componentLinearAccScale {z=0.44}
    componentLinearVelScale {z=0.44}
    physicsBodyDefinitions [{
      collisionType=CustomSphere
      sphereCollisionRadius=1
      boundBone {boneName=dyn_l_braid_02}
      boxExtents=<22 30 30>
      constraintSetup {angularLimitsMax=<45 0 45>, angularLimitsMin=<-45 -45 -45>, angularTarget=<0 0 0>}
      localJointOffset=<10 -5 0>
     }]
    sphericalLimits [
     {limitRadius=65.9176, drivingBone {boneName=spine_05}, sphereLocalOffset=<-12.8628 53 0>}
     {limitRadius=50, drivingBone {boneName=pelvis}, sphereLocalOffset=<-2.6826 40 0>}
     {limitRadius=28, drivingBone {boneName=spine_02}, sphereLocalOffset=<2.6549 15 0>}
    ]
   }
   n92 = Anim Dynamics - Bone: dyn_l_braid_03(ComponentPose=@n91, Alpha=@n89, GravityOverride="0,0,0", ExternalForce="0,0,0") {
    angularDampingOverride=0.9
    angularSpringConstant=2500
    bAngularSpring=true
    bGravityOverrideInSimSpace=true
    bOverrideAngularDamping=true
    bUseGravityOverride=true
    bUseSphericalLimits=true
    gravityScale=0.35
    simulationSpace=RootRelative
    alphaBoolBlend {blendOption=Linear}
    alphaScaleBiasClamp {scale=0.35}
    boundBone {boneName=dyn_l_braid_03}
    componentAppliedLinearAccClamp=<200 200 200>
    componentLinearAccScale {z=0.44}
    componentLinearVelScale {z=0.44}
    physicsBodyDefinitions [{
      collisionType=CustomSphere
      sphereCollisionRadius=1
      boundBone {boneName=dyn_l_braid_03}
      boxExtents=<22 30 30>
      constraintSetup {angularLimitsMax=<45 0 45>, angularLimitsMin=<-45 -45 -45>, angularTarget=<0 0 0>}
      localJointOffset=<10 -5 0>
     }]
    sphericalLimits [
     {limitRadius=65.9176, drivingBone {boneName=spine_05}, sphereLocalOffset=<-12.8628 52 0>}
     {limitRadius=50, drivingBone {boneName=pelvis}, sphereLocalOffset=<-2.6826 38.079 0>}
     {limitRadius=28, drivingBone {boneName=spine_02}, sphereLocalOffset=<2.6549 18 0>}
    ]
   }
   n93 = Anim Dynamics - Bone: dyn_l_braid_04(ComponentPose=@n92, Alpha=1, GravityOverride="0,0,0", ExternalForce="0,0,0") {
    angularDampingOverride=0.9
    angularSpringConstant=2500
    bAngularSpring=true
    bGravityOverrideInSimSpace=true
    bOverrideAngularDamping=true
    bUseGravityOverride=true
    bUseSphericalLimits=true
    gravityScale=0.35
    simulationSpace=RootRelative
    alphaBoolBlend {blendOption=Linear}
    alphaScaleBiasClamp {scale=0.35}
    boundBone {boneName=dyn_l_braid_04}
    componentAppliedLinearAccClamp=<200 200 200>
    componentLinearAccScale {z=0.44}
    componentLinearVelScale {z=0.44}
    physicsBodyDefinitions [{
      collisionType=CustomSphere
      sphereCollisionRadius=1
      boundBone {boneName=dyn_l_braid_04}
      boxExtents=<10 15 15>
      constraintSetup {angularLimitsMax=<45 0 45>, angularLimitsMin=<-45 -45 -45>, angularTarget=<0 0 0>}
      localJointOffset=<10 -5 0>
     }]
    sphericalLimits [
     {limitRadius=65.9176, drivingBone {boneName=spine_05}, sphereLocalOffset=<-12.8628 52 0>}
     {limitRadius=50, drivingBone {boneName=pelvis}, sphereLocalOffset=<-2.6826 38.079 0>}
     {limitRadius=28, drivingBone {boneName=spine_02}, sphereLocalOffset=<2.6549 15 0>}
     {limitRadius=1, drivingBone {boneName=pelvis}, sphereLocalOffset=<-30 0 0>}
    ]
   }
   n94 = Anim Dynamics - Bone: dyn_l_braid_05(ComponentPose=@n93, Alpha=1, GravityOverride="0,0,0", ExternalForce="0,0,0") {
    angularDampingOverride=0.9
    angularSpringConstant=2500
    bAngularSpring=true
    bGravityOverrideInSimSpace=true
    bOverrideAngularDamping=true
    bUseGravityOverride=true
    bUseSphericalLimits=true
    gravityScale=0.35
    simulationSpace=RootRelative
    alphaBoolBlend {blendOption=Linear}
    alphaScaleBiasClamp {scale=0.35}
    boundBone {boneName=dyn_l_braid_05}
    componentAppliedLinearAccClamp=<200 200 200>
    componentLinearAccScale {z=0.44}
    componentLinearVelScale {z=0.44}
    physicsBodyDefinitions [{
      collisionType=CustomSphere
      sphereCollisionRadius=1
      boundBone {boneName=dyn_l_braid_05}
      boxExtents=<10 15 15>
      constraintSetup {angularLimitsMax=<45 0 45>, angularLimitsMin=<-45 -45 -45>, angularTarget=<0 0 0>}
      localJointOffset=<10 -5 0>
     }]
    sphericalLimits [
     {limitRadius=65.9176, drivingBone {boneName=spine_05}, sphereLocalOffset=<-12.8628 52 0>}
     {limitRadius=50, drivingBone {boneName=pelvis}, sphereLocalOffset=<-2.6826 38.079 0>}
     {limitRadius=28, drivingBone {boneName=spine_02}, sphereLocalOffset=<2.6549 15 0>}
     {limitRadius=1, drivingBone {boneName=pelvis}, sphereLocalOffset=<-30 0 0>}
    ]
   }
   n95 = Anim Dynamics - Bone: dyn_l_braid_06(ComponentPose=@n94, Alpha=1, GravityOverride="0,0,0", ExternalForce="0,0,0") {
    angularDampingOverride=0.9
    angularSpringConstant=2500
    bAngularSpring=true
    bGravityOverrideInSimSpace=true
    bOverrideAngularDamping=true
    bUseGravityOverride=true
    bUseSphericalLimits=true
    gravityScale=0.35
    simulationSpace=RootRelative
    alphaBoolBlend {blendOption=Linear}
    alphaScaleBiasClamp {scale=0.35}
    boundBone {boneName=dyn_l_braid_06}
    componentAppliedLinearAccClamp=<200 200 200>
    componentLinearAccScale {z=0.44}
    componentLinearVelScale {z=0.44}
    physicsBodyDefinitions [{
      collisionType=CustomSphere
      sphereCollisionRadius=1
      boundBone {boneName=dyn_l_braid_06}
      boxExtents=<10 15 15>
      constraintSetup {angularLimitsMax=<45 0 45>, angularLimitsMin=<-45 -45 -45>, angularTarget=<0 0 0>}
      localJointOffset=<10 -5 0>
     }]
    sphericalLimits [
     {limitRadius=65.9176, drivingBone {boneName=spine_05}, sphereLocalOffset=<-12.8628 52 0>}
     {limitRadius=50, drivingBone {boneName=pelvis}, sphereLocalOffset=<-2.6826 38.079 0>}
     {limitRadius=28, drivingBone {boneName=spine_02}, sphereLocalOffset=<2.6549 15 0>}
     {limitRadius=1, drivingBone {boneName=pelvis}, sphereLocalOffset=<-30 0 0>}
    ]
   }
   n96 = Anim Dynamics - Bone: dyn_l_braid_07(ComponentPose=@n95, Alpha=1, GravityOverride="0,0,0", ExternalForce="0,0,0") {
    angularDampingOverride=0.9
    angularSpringConstant=2500
    bAngularSpring=true
    bGravityOverrideInSimSpace=true
    bOverrideAngularDamping=true
    bUseGravityOverride=true
    bUseSphericalLimits=true
    gravityScale=0.35
    simulationSpace=RootRelative
    alphaBoolBlend {blendOption=Linear}
    alphaScaleBiasClamp {scale=0.4}
    boundBone {boneName=dyn_l_braid_07}
    componentAppliedLinearAccClamp=<200 200 200>
    componentLinearAccScale {z=0.46}
    componentLinearVelScale {z=0.46}
    physicsBodyDefinitions [{
      collisionType=CustomSphere
      sphereCollisionRadius=1
      boundBone {boneName=dyn_l_braid_07}
      boxExtents=<10 15 15>
      constraintSetup {angularLimitsMax=<45 0 45>, angularLimitsMin=<-45 -45 -45>, angularTarget=<0 0 0>}
      localJointOffset=<10 -5 0>
     }]
    sphericalLimits [
     {limitRadius=65.9176, drivingBone {boneName=spine_05}, sphereLocalOffset=<-12.8628 62 0>}
     {limitRadius=50, drivingBone {boneName=pelvis}, sphereLocalOffset=<-2.6826 38.079 0>}
     {limitRadius=28, drivingBone {boneName=spine_02}, sphereLocalOffset=<2.6549 15 0>}
     {limitRadius=1, drivingBone {boneName=pelvis}, sphereLocalOffset=<-30 0 0>}
    ]
   }
   n97 = Anim Dynamics - Bone: dyn_l_braid_08(ComponentPose=@n96, Alpha=1, GravityOverride="0,0,0", ExternalForce="0,0,0") {
    angularDampingOverride=0.9
    angularSpringConstant=2500
    bAngularSpring=true
    bGravityOverrideInSimSpace=true
    bOverrideAngularDamping=true
    bUseGravityOverride=true
    bUseSphericalLimits=true
    gravityScale=0.35
    simulationSpace=RootRelative
    alphaBoolBlend {blendOption=Linear}
    alphaScaleBiasClamp {scale=0.5}
    boundBone {boneName=dyn_l_braid_08}
    componentAppliedLinearAccClamp=<200 200 200>
    componentLinearAccScale {z=0.46}
    componentLinearVelScale {z=0.46}
    physicsBodyDefinitions [{
      collisionType=CustomSphere
      sphereCollisionRadius=1
      boundBone {boneName=dyn_l_braid_08}
      boxExtents=<10 15 15>
      constraintSetup {angularLimitsMax=<45 0 45>, angularLimitsMin=<-45 -45 -45>, angularTarget=<0 0 0>}
      localJointOffset=<10 -5 0>
     }]
    sphericalLimits [
     {limitRadius=65.9176, drivingBone {boneName=spine_05}, sphereLocalOffset=<-12.8628 62 0>}
     {limitRadius=50, drivingBone {boneName=pelvis}, sphereLocalOffset=<-2.6826 38.079 0>}
     {limitRadius=28, drivingBone {boneName=spine_02}, sphereLocalOffset=<2.6549 15 0>}
     {limitRadius=1, drivingBone {boneName=pelvis}, sphereLocalOffset=<-30 0 0>}
    ]
   }
   n98 = Anim Dynamics - Bone: dyn_r_braid_01(ComponentPose=@n97, Alpha=1, GravityOverride="0,0,0", ExternalForce="0,0,0") {
    angularDampingOverride=1
    angularSpringConstant=2500
    bAngularSpring=true
    bGravityOverrideInSimSpace=true
    bOverrideAngularDamping=true
    bUseGravityOverride=true
    bUseSphericalLimits=true
    gravityScale=0.35
    simulationSpace=RootRelative
    alphaBoolBlend {blendOption=Linear}
    boundBone {boneName=dyn_r_braid_01}
    componentAppliedLinearAccClamp=<200 200 200>
    componentLinearAccScale {z=0.44}
    componentLinearVelScale {z=0.44}
    physicsBodyDefinitions [{
      collisionType=CustomSphere
      sphereCollisionRadius=1
      boundBone {boneName=dyn_r_braid_01}
      boxExtents=<22 30 30>
      constraintSetup {angularLimitsMax=<45 45 45>, angularLimitsMin=<-45 -45 -45>, angularTarget=<0 0 0>}
      localJointOffset=<10 -5 0>
     }]
    sphericalLimits [
     {limitRadius=65.9176, drivingBone {boneName=spine_05}, sphereLocalOffset=<-12.8628 53 0>}
     {limitRadius=50, drivingBone {boneName=pelvis}, sphereLocalOffset=<-2.6826 40 0>}
     {limitRadius=28, drivingBone {boneName=spine_02}, sphereLocalOffset=<2.6549 15 0>}
    ]
   }
   n99 = Anim Dynamics - Bone: dyn_r_braid_02(ComponentPose=@n98, Alpha=1, GravityOverride="0,0,0", ExternalForce="0,0,0") {
    angularDampingOverride=1
    angularSpringConstant=2500
    bAngularSpring=true
    bGravityOverrideInSimSpace=true
    bOverrideAngularDamping=true
    bUseGravityOverride=true
    bUseSphericalLimits=true
    gravityScale=0.35
    simulationSpace=RootRelative
    alphaBoolBlend {blendOption=Linear}
    alphaScaleBiasClamp {scale=0.5}
    boundBone {boneName=dyn_r_braid_02}
    componentAppliedLinearAccClamp=<200 200 200>
    componentLinearAccScale {z=0.44}
    componentLinearVelScale {z=0.44}
    physicsBodyDefinitions [{
      collisionType=CustomSphere
      sphereCollisionRadius=1
      boundBone {boneName=dyn_r_braid_02}
      boxExtents=<22 30 30>
      constraintSetup {angularLimitsMax=<45 45 45>, angularLimitsMin=<-45 -45 -45>, angularTarget=<0 0 0>}
      localJointOffset=<10 -5 0>
     }]
    sphericalLimits [
     {limitRadius=65.9176, drivingBone {boneName=spine_05}, sphereLocalOffset=<-12.8628 53 0>}
     {limitRadius=50, drivingBone {boneName=pelvis}, sphereLocalOffset=<-2.6826 40 0>}
     {limitRadius=28, drivingBone {boneName=spine_02}, sphereLocalOffset=<2.6549 15 0>}
    ]
   }
   n100 = Anim Dynamics - Bone: dyn_r_braid_03(ComponentPose=@n99, Alpha=1, GravityOverride="0,0,0", ExternalForce="0,0,0") {
    angularDampingOverride=0.9
    angularSpringConstant=2500
    bAngularSpring=true
    bGravityOverrideInSimSpace=true
    bOverrideAngularDamping=true
    bUseGravityOverride=true
    bUseSphericalLimits=true
    gravityScale=0.35
    simulationSpace=RootRelative
    alphaBoolBlend {blendOption=Linear}
    alphaScaleBiasClamp {scale=0.35}
    boundBone {boneName=dyn_r_braid_03}
    componentAppliedLinearAccClamp=<200 200 200>
    componentLinearAccScale {z=0.44}
    componentLinearVelScale {z=0.44}
    physicsBodyDefinitions [{
      collisionType=CustomSphere
      sphereCollisionRadius=1
      boundBone {boneName=dyn_r_braid_03}
      boxExtents=<22 30 30>
      constraintSetup {angularLimitsMax=<45 45 45>, angularLimitsMin=<-45 -45 -45>, angularTarget=<0 0 0>}
      localJointOffset=<10 -5 0>
     }]
    sphericalLimits [
     {limitRadius=65.9176, drivingBone {boneName=spine_05}, sphereLocalOffset=<-12.8628 52 0>}
     {limitRadius=50, drivingBone {boneName=pelvis}, sphereLocalOffset=<-2.6826 38.079 0>}
     {limitRadius=28, drivingBone {boneName=spine_02}, sphereLocalOffset=<2.6549 18 0>}
    ]
   }
   n101 = Anim Dynamics - Bone: dyn_r_braid_04(ComponentPose=@n100, Alpha=1, GravityOverride="0,0,0", ExternalForce="0,0,0") {
    angularDampingOverride=0.9
    angularSpringConstant=2500
    bAngularSpring=true
    bGravityOverrideInSimSpace=true
    bOverrideAngularDamping=true
    bUseGravityOverride=true
    bUseSphericalLimits=true
    gravityScale=0.35
    simulationSpace=RootRelative
    alphaBoolBlend {blendOption=Linear}
    alphaScaleBiasClamp {scale=0.35}
    boundBone {boneName=dyn_r_braid_04}
    componentAppliedLinearAccClamp=<200 200 200>
    componentLinearAccScale {z=0.44}
    componentLinearVelScale {z=0.44}
    physicsBodyDefinitions [{
      collisionType=CustomSphere
      sphereCollisionRadius=1
      boundBone {boneName=dyn_r_braid_04}
      boxExtents=<10 15 15>
      constraintSetup {angularLimitsMax=<45 45 45>, angularLimitsMin=<-45 -45 -45>, angularTarget=<0 0 0>}
      localJointOffset=<10 -5 0>
     }]
    sphericalLimits [
     {limitRadius=65.9176, drivingBone {boneName=spine_05}, sphereLocalOffset=<-12.8628 52 0>}
     {limitRadius=50, drivingBone {boneName=pelvis}, sphereLocalOffset=<-2.6826 38.079 0>}
     {limitRadius=28, drivingBone {boneName=spine_02}, sphereLocalOffset=<2.6549 15 0>}
     {limitRadius=1, drivingBone {boneName=pelvis}, sphereLocalOffset=<-30 0 0>}
    ]
   }
   n102 = Anim Dynamics - Bone: dyn_r_braid_05(ComponentPose=@n101, Alpha=1, GravityOverride="0,0,0", ExternalForce="0,0,0") {
    angularDampingOverride=0.9
    angularSpringConstant=2500
    bAngularSpring=true
    bGravityOverrideInSimSpace=true
    bOverrideAngularDamping=true
    bUseGravityOverride=true
    bUseSphericalLimits=true
    gravityScale=0.35
    simulationSpace=RootRelative
    alphaBoolBlend {blendOption=Linear}
    alphaScaleBiasClamp {scale=0.35}
    boundBone {boneName=dyn_r_braid_05}
    componentAppliedLinearAccClamp=<200 200 200>
    componentLinearAccScale {z=0.44}
    componentLinearVelScale {z=0.44}
    physicsBodyDefinitions [{
      collisionType=CustomSphere
      sphereCollisionRadius=1
      boundBone {boneName=dyn_r_braid_05}
      boxExtents=<10 15 15>
      constraintSetup {angularLimitsMax=<45 45 45>, angularLimitsMin=<-45 -45 -45>, angularTarget=<0 0 0>}
      localJointOffset=<10 -5 0>
     }]
    sphericalLimits [
     {limitRadius=65.9176, drivingBone {boneName=spine_05}, sphereLocalOffset=<-12.8628 52 0>}
     {limitRadius=50, drivingBone {boneName=pelvis}, sphereLocalOffset=<-2.6826 38.079 0>}
     {limitRadius=28, drivingBone {boneName=spine_02}, sphereLocalOffset=<2.6549 15 0>}
     {limitRadius=1, drivingBone {boneName=pelvis}, sphereLocalOffset=<-30 0 0>}
    ]
   }
   n103 = Anim Dynamics - Bone: dyn_r_braid_06(ComponentPose=@n102, Alpha=1, GravityOverride="0,0,0", ExternalForce="0,0,0") {
    angularDampingOverride=0.9
    angularSpringConstant=2500
    bAngularSpring=true
    bGravityOverrideInSimSpace=true
    bOverrideAngularDamping=true
    bUseGravityOverride=true
    bUseSphericalLimits=true
    gravityScale=0.35
    simulationSpace=RootRelative
    alphaBoolBlend {blendOption=Linear}
    alphaScaleBiasClamp {scale=0.35}
    boundBone {boneName=dyn_r_braid_06}
    componentAppliedLinearAccClamp=<200 200 200>
    componentLinearAccScale {z=0.44}
    componentLinearVelScale {z=0.44}
    physicsBodyDefinitions [{
      collisionType=CustomSphere
      sphereCollisionRadius=1
      boundBone {boneName=dyn_r_braid_06}
      boxExtents=<10 15 15>
      constraintSetup {angularLimitsMax=<45 45 45>, angularLimitsMin=<-45 -45 -45>, angularTarget=<0 0 0>}
      localJointOffset=<10 -5 0>
     }]
    sphericalLimits [
     {limitRadius=65.9176, drivingBone {boneName=spine_05}, sphereLocalOffset=<-12.8628 52 0>}
     {limitRadius=50, drivingBone {boneName=pelvis}, sphereLocalOffset=<-2.6826 38.079 0>}
     {limitRadius=28, drivingBone {boneName=spine_02}, sphereLocalOffset=<2.6549 15 0>}
     {limitRadius=1, drivingBone {boneName=pelvis}, sphereLocalOffset=<-30 0 0>}
    ]
   }
   n104 = Anim Dynamics - Bone: dyn_r_braid_07(ComponentPose=@n103, Alpha=1, GravityOverride="0,0,0", ExternalForce="0,0,0") {
    angularDampingOverride=0.9
    angularSpringConstant=2500
    bAngularSpring=true
    bGravityOverrideInSimSpace=true
    bOverrideAngularDamping=true
    bUseGravityOverride=true
    bUseSphericalLimits=true
    gravityScale=0.35
    simulationSpace=RootRelative
    alphaBoolBlend {blendOption=Linear}
    alphaScaleBiasClamp {scale=0.4}
    boundBone {boneName=dyn_r_braid_07}
    componentAppliedLinearAccClamp=<200 200 200>
    componentLinearAccScale {z=0.46}
    componentLinearVelScale {z=0.46}
    physicsBodyDefinitions [{
      collisionType=CustomSphere
      sphereCollisionRadius=1
      boundBone {boneName=dyn_r_braid_07}
      boxExtents=<10 15 15>
      constraintSetup {angularLimitsMax=<45 45 45>, angularLimitsMin=<-45 -45 -45>, angularTarget=<0 0 0>}
      localJointOffset=<10 -5 0>
     }]
    sphericalLimits [
     {limitRadius=65.9176, drivingBone {boneName=spine_05}, sphereLocalOffset=<-12.8628 62 0>}
     {limitRadius=50, drivingBone {boneName=pelvis}, sphereLocalOffset=<-2.6826 38.079 0>}
     {limitRadius=28, drivingBone {boneName=spine_02}, sphereLocalOffset=<2.6549 15 0>}
     {limitRadius=1, drivingBone {boneName=pelvis}, sphereLocalOffset=<-30 0 0>}
    ]
   }
   n105 = Anim Dynamics - Bone: dyn_r_braid_08(ComponentPose=@n104, Alpha=1, GravityOverride="0,0,0", ExternalForce="0,0,0") {
    angularDampingOverride=0.9
    angularSpringConstant=2500
    bAngularSpring=true
    bGravityOverrideInSimSpace=true
    bOverrideAngularDamping=true
    bUseGravityOverride=true
    bUseSphericalLimits=true
    gravityScale=0.35
    simulationSpace=RootRelative
    alphaBoolBlend {blendOption=Linear}
    alphaScaleBiasClamp {scale=0.5}
    boundBone {boneName=dyn_r_braid_08}
    componentAppliedLinearAccClamp=<200 200 200>
    componentLinearAccScale {z=0.46}
    componentLinearVelScale {z=0.46}
    physicsBodyDefinitions [{
      collisionType=CustomSphere
      sphereCollisionRadius=1
      boundBone {boneName=dyn_r_braid_08}
      boxExtents=<10 15 15>
      constraintSetup {angularLimitsMax=<45 45 45>, angularLimitsMin=<-45 -45 -45>, angularTarget=<0 0 0>}
      localJointOffset=<10 -5 0>
     }]
    sphericalLimits [
     {limitRadius=65.9176, drivingBone {boneName=spine_05}, sphereLocalOffset=<-12.8628 62 0>}
     {limitRadius=50, drivingBone {boneName=pelvis}, sphereLocalOffset=<-2.6826 38.079 0>}
     {limitRadius=28, drivingBone {boneName=spine_02}, sphereLocalOffset=<2.6549 15 0>}
     {limitRadius=1, drivingBone {boneName=pelvis}, sphereLocalOffset=<-30 0 0>}
    ]
   }
   n106 = "Constraint - dyn_l_braid_01 (P)"(ComponentPose=@n105, Alpha=1, ConstraintWeights_0=1) {
    alphaBoolBlend {blendOption=Linear}
    boneToModify {boneName=dyn_l_braid_01}
    constraintSetup [{targetBone {boneName=spine_05}}]
    constraintWeights [1]
   }
   n107 = "Constraint - dyn_l_braid_02 (P)"(ComponentPose=@n106, Alpha=0, ConstraintWeights_0=1) {
    alphaBoolBlend {blendOption=Linear}
    boneToModify {boneName=dyn_l_braid_02}
    constraintSetup [{targetBone {boneName=spine_05}}]
    constraintWeights [1]
   }
   n108 = "Constraint - dyn_l_braid_03 (P)"(ComponentPose=@n107, Alpha=0, ConstraintWeights_0=1) {
    alphaBoolBlend {blendOption=Linear}
    boneToModify {boneName=dyn_l_braid_03}
    constraintSetup [{targetBone {boneName=spine_05}}]
    constraintWeights [1]
   }
   n109 = "Constraint - dyn_r_braid_01 (P)"(ComponentPose=@n108, Alpha=1, ConstraintWeights_0=1) {
    alphaBoolBlend {blendOption=Linear}
    boneToModify {boneName=dyn_r_braid_01}
    constraintSetup [{targetBone {boneName=spine_05}}]
    constraintWeights [1]
   }
   n110 = "Constraint - dyn_r_braid_02 (P)"(ComponentPose=@n109, Alpha=0, ConstraintWeights_0=1) {
    alphaBoolBlend {blendOption=Linear}
    boneToModify {boneName=dyn_r_braid_02}
    constraintSetup [{targetBone {boneName=spine_05}}]
    constraintWeights [1]
   }
   n111 = "Constraint - dyn_r_braid_03 (P)"(ComponentPose=@n110, Alpha=0, ConstraintWeights_0=1) {
    alphaBoolBlend {blendOption=Linear}
    boneToModify {boneName=dyn_r_braid_03}
    constraintSetup [{targetBone {boneName=spine_05}}]
    constraintWeights [1]
   }
   n112 = "Transform (Modify) Bone - Bone: dyn_l_braid_MSTR"(ComponentPose=@n111, Alpha=1, Translation="0,0,0", Rotation="30,0,0", Scale="1,1,1") {
    rotationMode=BMM_Additive
    rotationSpace=BCS_BoneSpace
    scaleMode=BMM_Additive
    scaleSpace=BCS_BoneSpace
    translationMode=BMM_Additive
    translationSpace=BCS_BoneSpace
    alphaBoolBlend {blendOption=Linear}
    boneToModify {boneName=dyn_l_braid_MSTR}
    rotation {pitch=30}
   }
   n113 = "Transform (Modify) Bone - Bone: dyn_r_braid_MSTR"(ComponentPose=@n112, Alpha=1, Translation="0,0,0", Rotation="30,0,0", Scale="1,1,1") {
    rotationMode=BMM_Additive
    rotationSpace=BCS_BoneSpace
    scaleMode=BMM_Additive
    scaleSpace=BCS_BoneSpace
    translationMode=BMM_Additive
    translationSpace=BCS_BoneSpace
    alphaBoolBlend {blendOption=Linear}
    boneToModify {boneName=dyn_r_braid_MSTR}
    rotation {pitch=30}
   }
   n114 = Get bIsSkydiveDiveMode()
   n115 = "Transform (Modify) Bone - Bone: dyn_l_braid_MSTR"!(ComponentPose=@n113, bAlphaBoolEnabled:bool=@n114, Translation="0,0,0", Rotation="-10,0,0", Scale="1,1,1") -> (Pose) {
    alphaInputType=Bool
    rotationMode=BMM_Additive
    rotationSpace=BCS_BoneSpace
    scaleMode=BMM_Additive
    scaleSpace=BCS_BoneSpace
    translationMode=BMM_Additive
    translationSpace=BCS_BoneSpace
    alphaBoolBlend {blendInTime=0.1, blendOption=Linear, blendOutTime=0.1}
    boneToModify {boneName=dyn_l_braid_MSTR}
    rotation {pitch=-10}
   }
   n116 = "Transform (Modify) Bone - Bone: dyn_r_braid_MSTR"!(ComponentPose=@n115, bAlphaBoolEnabled:bool=@n114, Translation="0,0,0", Rotation="-10,0,0", Scale="1,1,1") -> (Pose) {
    alphaInputType=Bool
    rotationMode=BMM_Additive
    rotationSpace=BCS_BoneSpace
    scaleMode=BMM_Additive
    scaleSpace=BCS_BoneSpace
    translationMode=BMM_Additive
    translationSpace=BCS_BoneSpace
    alphaBoolBlend {blendInTime=0.1, blendOption=Linear, blendOutTime=0.1}
    boneToModify {boneName=dyn_r_braid_MSTR}
    rotation {pitch=-10}
   }
   n117 = Blend Poses by bool(
    BlendPose_0=Component To Local(ComponentPose=@n116)
    BlendPose_1=Blend Poses by bool(
     BlendPose_0=Component To Local(ComponentPose=@n88)
     BlendPose_1=Blend Poses by bool(BlendPose_0=@n67, BlendPose_1=@n48, bActiveValue=@n68) {blendType=Linear}
     bActiveValue=Get BackInAir() -> (BackInAir:bool)
    ) {blendType=Linear}
    bActiveValue=Get WithoutParachuteInAir() -> (WithoutParachuteInAir:bool)
   ) {blendType=Linear}
   n118 = "Transform (Modify) Bone - Bone: dyn_l_braid_MSTR"!(
    ComponentPose=Local To Component(LocalPose=Use cached pose 'ControlRig'())
    Alpha:float=1
    Translation="0,0,3"
    Rotation="0,50,0"
    Scale="1,1,1"
   ) -> (Pose) {
    rotationMode=BMM_Additive
    rotationSpace=BCS_BoneSpace
    scaleMode=BMM_Additive
    scaleSpace=BCS_BoneSpace
    translationMode=BMM_Additive
    translationSpace=BCS_BoneSpace
    alphaBoolBlend {blendOption=Linear}
    alphaScaleBiasClamp {bInterpResult=true, interpSpeedDecreasing=20, interpSpeedIncreasing=20}
    boneToModify {boneName=dyn_l_braid_MSTR}
    rotation {yaw=50}
    translation {z=3}
   }
   n119 = "Transform (Modify) Bone - Bone: dyn_r_braid_MSTR"!(ComponentPose=@n118, Alpha:float=1, Translation="3,0,3", Rotation="0,50,0", Scale="1,1,1") -> (Pose) {
    rotationMode=BMM_Additive
    rotationSpace=BCS_BoneSpace
    scaleMode=BMM_Additive
    scaleSpace=BCS_BoneSpace
    translationMode=BMM_Additive
    translationSpace=BCS_BoneSpace
    alphaBoolBlend {blendOption=Linear}
    alphaScaleBiasClamp {bInterpResult=true, interpSpeedDecreasing=20, interpSpeedIncreasing=20}
    boneToModify {boneName=dyn_r_braid_MSTR}
    rotation {yaw=50}
    translation {x=3, z=3}
   }
   n120 = RigidBody!(ComponentPose=@n119, Alpha:float, OverrideWorldGravity=Get Gravity_Override(), ExternalForce) -> (Pose) {
    bClampLinearTranslationLimitToRefPose=true
    bOverrideWorldGravity=true
    overridePhysicsAsset=&F_MED_Quarrel_FaceAcc_FrontEnd_Physics
    simulationSpace=BaseBoneSpace
    alphaBoolBlend {blendOption=Linear}
    baseBoneRef {boneName=root}
    componentAppliedLinearAccClamp=<2500 2500 2500>
    componentLinearAccScale=<0.3 0.3 0.3>
    componentLinearVelScale=<0.3 0.3 0.3>
    overrideWorldGravity {z=-980}
   }
   n121 = Component To Local(ComponentPose=@n120)
   n122 = Get bIsDBNOCarried()
   n123 = Anim Dynamics - Bone: dyn_l_hatFlap(
    ComponentPose=Local To Component(LocalPose=Blend Poses by bool(BlendPose_0=@n121, BlendPose_1=@n117, bActiveValue=@n122) {blendType=Linear})
    Alpha=0.5
    ExternalForce="0,0,0"
   ) {
    angularDampingOverride=0.9999
    angularSpringConstant=50
    bAngularSpring=true
    bGravityOverrideInSimSpace=true
    bOverrideAngularDamping=true
    gravityScale=0.5
    linearDampingOverride=3
    numSolverIterationsPreUpdate=3
    simulationSpace=RootRelative
    alphaBoolBlend {blendOption=Linear}
    boundBone {boneName=dyn_l_hatFlap}
    componentLinearAccScale=<1 1 1.1>
    componentLinearVelScale=<1 1 1.1>
    physicsBodyDefinitions [{
      sphereCollisionRadius=0
      boundBone {boneName=dyn_l_hatFlap}
      boxExtents=<10 20 20>
      constraintSetup {angularLimitsMax=<5 5 5>, angularLimitsMin=<-5 -50 -5>, angularTarget=<0 0 0>}
      localJointOffset=<0 0 -2>
     }]
   }
   n124 = RigidBody!(ComponentPose=@n123, Alpha=0.25, ExternalForce) -> (Pose) {
    bClampLinearTranslationLimitToRefPose=true
    bOverrideWorldGravity=true
    overridePhysicsAsset=&/F_MED_Quarrel_FaceAcc_Hair_Physics
    simulationSpace=BaseBoneSpace
    alphaBoolBlend {blendOption=Linear}
    baseBoneRef {boneName=root}
    componentAppliedLinearAccClamp=<5000 5000 5000>
    componentLinearAccScale=<0.3 0.3 0.45>
    componentLinearVelScale=<0.3 0.3 0.45>
    overrideWorldGravity {z=-980}
   }
   n125 = RigidBody(ComponentPose=@n124) {
    bClampLinearTranslationLimitToRefPose=true
    bOverrideWorldGravity=true
    overridePhysicsAsset=&/F_MED_Quarrel_FaceAcc_HairLongHair_Physics
    simulationSpace=BaseBoneSpace
    alphaBoolBlend {blendOption=Linear}
    baseBoneRef {boneName=root}
    componentAppliedLinearAccClamp=<500 500 1500>
    componentLinearAccScale=<0.25 0.25 0.5>
    componentLinearVelScale=<0.25 0.25 0.45>
    overrideWorldGravity {z=-980}
   }
   Save cached pose 'PostSim'(Pose=Component To Local(ComponentPose=@n125)) {cacheName=PostSim}
   n126 = Get Body_Mesh()
   n127 = Layered blend per bone(
    BasePose=Copy Pose From Mesh(SourceMeshComponent=@n126) {bCopyCurves=true, bUseAttachedParent=true}
    BlendPoses_0=Copy Pose From Mesh(SourceMeshComponent=@n126) {bCopyCurves=true, bUseAttachedParent=true}
   ) {layerSetup [{branchFilters [{boneName=thigh_l}, {boneName=thigh_r}, {boneName=calf_l}, {boneName=calf_r}]}]}
   Save cached pose 'ControlRig'(Pose=@n127) {cacheName=ControlRig}
  ]}
 EventGraph {nodes [
   Event OnCharacterCustomizationLoaded()
   > Delay(Duration=0.1)
   > ResetDynamics(InTeleportType=TeleportPhysics)
   > ResetDynamics(InTeleportType=ResetPhysics)
   Event BlueprintInitializeAnimation()
   > Set Body_Mesh(Body_Mesh=GetOwnerSkeletalMeshComponent())
   > n1 = Cast To PlayerPawn_Athena(Object=TryGetPawnOwner())
   > n2 = Set Pawn Owner(Pawn Owner:Player Pawn Athena Object Reference=@n1) -> (Output_Get:Player Pawn Athena Object Reference)
   > Set Head_Mesh(Head_Mesh:Skeletal Mesh Component Object Reference=GetSkeletalMeshForPartType(self=@n2, PartType=Head)) -> (Output_Get:Skeletal Mesh Component Object Reference)
   Event OnEndCrouching()
   > Set crouch_alpha(crouch_alpha:double=0) -> (Output_Get:double)
   Event OnBeginCrouching()
   > Set crouch_alpha(crouch_alpha=1)
   Event OnEndSkydiving()
   > Set Skydiving(Skydiving:double=0) -> (Output_Get:double)
   Event OnBeginSkydiving()
   > Set Skydiving(Skydiving=1)
   Event OnBeginPlayingEmote() {is=title K2Node_Event}
   > Set PlayingEmote(PlayingEmote:double=0.3) -> (Output_Get:double)
   Event OnEndPlayingEmote() {is=title K2Node_Event}
   > Set PlayingEmote(PlayingEmote=0)
   Event BlueprintUpdateAnimation()
   > n3 = Sequence!(execute) -> (then_0, then_1, then_2:Exec, then_3:Exec, then_4:Exec, then_5:Exec, then_6:Exec, then_7:Exec, then_8:Exec)
   n4 = Cast To PlayerPawn_Athena(execute=@n3.then_2, Object=TryGetPawnOwner())
   n5 = GetOwnerSkeletalMeshComponent()
   n6 = BreakRotator(InRot=BreakTransform(InTransform=GetDeltaTransformFromRefPose(self=@n5, BoneName=head, BaseName=neck_02)).Rotation)
   Set LookUp(execute=@n3.then_3, LookUp:double=MapRangeClamped(Value="Absolute (Float)"(A=@n6.Yaw), InRangeA=0, InRangeB=-17)) -> (Output_Get:double)
   n7 = BreakTransform(InTransform=GetDeltaTransformFromRefPose(self=@n5, BoneName=upperarm_r, BaseName=clavicle_r))
   > Set LookUpArms(LookUpArms:double=MapRangeClamped(Value=BreakRotator(InRot=@n7.Rotation).Yaw, InRangeA=90, InRangeB=120)) -> (Output_Get:double)
   n8 = BreakTransform(InTransform=GetDeltaTransformFromRefPose(self=@n5, BoneName=upperarm_r, BaseName=clavicle_r))
   > Set LookUpRifle(LookUpRifle:double=MapRangeClamped(Value=BreakRotator(InRot=@n8.Rotation).Yaw, InRangeA=60, InRangeB=75)) -> (Output_Get:double)
   n9 = GetAnimInstance(self=GetOwnerSkeletalMeshComponent())
   Is Valid(exec=@n3.then_5, InputObject=@n9)
   n10 = Montage_IsPlaying(self=@n9, Montage=/BRCosmetics/Animation/Game/MainPlayer/Emotes/QuarrelFemale/Emote_QuarrelFemale_CMF_M)
   > Set isCustomEmotePlaying(isCustomEmotePlaying:double=SelectFloat(A=0, B=1, bPickA=@n10)) -> (Output_Get:double)
   > Set isCustomEmotePlaying2(isCustomEmotePlaying2:double=SelectFloat(A=1, B=0, bPickA=@n10)) -> (Output_Get:double)
   n11 = CheckForBackbling()
   n12 = NOT Boolean(A=@n11.bIsBackblingLong)
   Set isWearingBackpack(execute=@n3.then_7, isWearingBackpack:double="To Float (Boolean)"(InBool=AND Boolean(A=@n11, B=@n12))) -> (Output_Get:double)
   > Set isNotWearingBackbling(isNotWearingBackbling:double="To Float (Boolean)"(InBool=NOT Boolean(A=@n11))) -> (Output_Get:double)
   > Set isWearingCape(isWearingCape:double="To Float (Boolean)"(InBool=OR Boolean(A=@n11.bIsBackblingACape, B=false))) -> (Output_Get:double)
   > Set WearingCape(WearingCape:bool=AND Boolean(A=@n11, B=@n11.bIsBackblingACape)) -> (Output_Get:bool)
   > n13 = Branch(Condition=@n11.bIsBackblingLong)
   > Set ACC(ACC:Vector="0, 0, 0") -> (Output_Get:Vector)
   > Set VEL(VEL:Vector="0, 0, 0") -> (Output_Get:Vector)
   > Set backpack(backpack:double=0.5) -> (Output_Get:double)
   Set ACC(execute=@n13.else, ACC="0.3,0.3,0.3")
   > Set VEL(VEL="0.3,0.3,0.3")
   > Set backpack(backpack=0)
   n14 = Get WearingCape() -> (WearingCape:bool)
   n15 = Get bIsInFrontEnd()
   Set isWearingCapeFE(execute=@n3.then_8, isWearingCapeFE:bool=NOT Boolean(A=AND Boolean(A=@n15, B=@n14))) -> (Output_Get:bool)
   n16 = Get bRigidBodyEnabled()
   n17 = Get bAnimDynamicsEnabled() -> (bAnimDynamicsEnabled:bool)
   > Set NoPhysic(NoPhysic:bool=AND Boolean(A=NOT Boolean(A=@n17), B=NOT Boolean(A=@n16))) -> (Output_Get:bool)
   > Set WearingCapeAlpha(WearingCapeAlpha:double=MapRangeClamped(Value=Get isWearingCape(), InRangeA=0, InRangeB=1, OutRangeA=1, OutRangeB=0.2)) -> (Output_Get:double)
   > Set BackInAir(BackInAir:bool=AND Boolean(A=Get bIsSkyDiving(), B=Get isWearingBackbling())) -> (Output_Get:bool)
   n18 = Get bIsParachuteOpened()
   > Set WithoutParachuteInAir(WithoutParachuteInAir:bool=AND Boolean(A=Get bIsSkyDiving(), B=NOT Boolean(A=@n18))) -> (Output_Get:bool)
   n19 = Gravity Override State Function DEPRECATED(
    execute=@n3.then_4
    CharacterPartAnimInstance=Self-Reference()
    GameTimeInSeconds=GetGameTimeInSeconds()
    GravityOverrideParams=Get Gravity Override Params()
   )
   > Set Gravity_Override(Gravity_Override:Vector=@n19) -> (Output_Get:Vector)
   n20 = GetOwnerSkeletalMeshComponent()
   n21 = BreakRotator(InRot=BreakTransform(InTransform=GetDeltaTransformFromRefPose(self=@n20, BoneName=head, BaseName=spine_05)).Rotation)
   Set head_rot_back(execute=@n3.then_1, head_rot_back:double=MapRangeClamped(Value=@n21.Yaw, InRangeA=0, InRangeB=-50)) -> (Output_Get:double)
   > Set head_rot_left(head_rot_left:double=MapRangeClamped(Value=@n21, InRangeA=0, InRangeB=30)) -> (Output_Get:double)
   > Set head_rot_right(head_rot_right:double=MapRangeClamped(Value=@n21, InRangeA=0, InRangeB=-30)) -> (Output_Get:double)
   > Set head_rot_fwd(head_rot_fwd:double=MapRangeClamped(Value=@n21.Yaw, InRangeA=0, InRangeB=25)) -> (Output_Get:double)
   n22 = Get bIsSkyDiving()
   Set skydive_alpha(execute=@n3.then_0, skydive_alpha:double="To Float (Boolean)"(InBool=@n22)) -> (Output_Get:double)
   > n23 = Branch(Condition=@n22)
   n24 = Get SkydivingExternalForceInput()
   > Set skydiving_force(skydiving_force:Vector=vector * float(
     A=Break Fort Anim Input Skydiving External Force(FortAnimInput_SkydivingExternalForce=@n24) -> (
      bUseSkydivingVectorForce=~
      SkydivingVectorForce=~
      HeadToPelvisDirection=~
      FloatingMultiplier=~
      FloatingAdditive=~
      DivingMultiplier=~
      DivingAdditive=~
      ParachutingMultiplier=~
      ParachutingAdditive=~
     ).SkydivingVectorForce
     B=1
    )) -> (Output_Get:Vector)
   Set skydiving_force(execute=@n23.else, skydiving_force="0, 0, 0")
   GetSkeletalMeshForPartType(self=@n4)
   Get bIsPlayingFullBodySlotInFrontEnd() -> (bIsPlayingFullBodySlotInFrontEnd:bool)
  ]}
]
variables [
 ACC {type=Vector}
 BackInAir {type=boolean}
 backpack {type=number}
 Body_Mesh {type=SkeletalMeshComponent}
 crouch_alpha {type=number}
 Gravity Override Params {
  type=GravityOverrideParamsStruct
  default {
   skydive_DivingDown {gravityOverride=<0 0 0>, windOffset=<0 -600 400>}
   skydive_FloatingDown {gravityOverride=<0 0 600>}
   skydive_GoingUp {gravityOverride=<0 0 0>, windOffset=<0 -600 400>}
   skydive_Parachuting {gravityOverride=<0 0 -980>, windOffset=<0 -600 400>}
   swimming {gravityOverride=<0 0 0>}
  }
 }
 Gravity_Override {type=Vector}
 Head_Mesh {type=SkeletalMeshComponent}
 head_rot_back {type=number}
 head_rot_fwd {type=number}
 head_rot_left {type=number}
 head_rot_right {type=number}
 isCustomEmotePlaying {type=number}
 isCustomEmotePlaying2 {type=number}
 isNotWearingBackbling {type=number}
 isWearingBackbling {type=boolean}
 isWearingBackpack {type=number}
 isWearingCape {type=number}
 isWearingCapeFE {type=boolean, default=true}
 LookUp {type=number}
 LookUpArms {type=number}
 LookUpRifle {type=number}
 NoPhysic {type=boolean, default=true}
 Pawn Owner {type=PlayerPawn_Athena_C}
 PlayingEmote {type=number}
 Skel_Mesh {type=SkeletalMeshComponent}
 skydive_alpha {type=number}
 Skydiving {type=number}
 skydiving_force {type=Vector}
 VEL {type=Vector}
 WearingCape {type=boolean, default=true}
 WearingCapeAlpha {type=number}
 WithoutParachuteInAir {type=boolean}
]
